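# Physics asset for the dangling strap loops on the left thigh of the female medium RainbowStraps body cosmetic.
# Three capsule-bodied bone chains (front-left, left, back-left) hang from master bodies on the left thigh twist bone, mostly with limited swing and tight twist.
# Angularly free constraints cross-link the lower loop segments, including a two-way back-to-side link.
# Capsules also cover both calves, the left thigh and both thigh twist bones.
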PhysicsAsset /BRCosmetics/Characters/Player/Female/Medium/Bodies/F_MED_RainbowStraps/Meshes/F_MED_RainbowStraps_Straps
bodies {
 calf_l [{shapeType=Capsule, length=40, radius=9, shapeName=calf_l_capsule, center=<-25 0 0>, rotation=<rot 90 0 0>}]
 calf_r [{shapeType=Capsule, length=40, radius=9, shapeName=calf_l_capsule, center=<25 0 0>, rotation=<rot 90 0 0>}]
 dyn_Loop_L_01 [{shapeType=Capsule, length=5, radius=5, shapeName=dyn_Loop_L_01_capsule, center=<5 0 6>, rotation=<rot 90 90 0>}]
 dyn_Loop_L_02 [{shapeType=Capsule, length=5, radius=5, shapeName=dyn_Loop_L_02_capsule, center=<5 0 3>, rotation=<rot 90 90 0>}]
 dyn_Loop_L_03 [{shapeType=Capsule, length=10, radius=6, shapeName=dyn_Loop_L_03_capsule, center=<5 0 2>, rotation=<rot 90 90 0>}]
 dyn_Loop_L_mstr [{shapeType=Capsule, length=1, radius=1, shapeName=dyn_Loop_L_mstr_capsule}]
 dyn_loop_BL_01 [{shapeType=Capsule, length=5, radius=4, shapeName=dyn_loop_BL_01_capsule, center=<4 0 3>, rotation=<rot 90 0 0>}]
 dyn_loop_BL_02 [{shapeType=Capsule, length=8, radius=4.5, shapeName=dyn_loop_BL_02_capsule, center=<7 0 3>, rotation=<rot 90 0 0>}]
 dyn_loop_BL_03 [{shapeType=Capsule, length=10, radius=5, shapeName=dyn_loop_BL_03_capsule, center=<7 0 2>, rotation=<rot 90 0 0>}]
 dyn_loop_BL_mstr [{shapeType=Capsule, length=1, radius=1, shapeName=dyn_loop_BL_mstr_capsule, rotation=<rot 0 -90 0>}]
 dyn_loop_FL_01 [{shapeType=Capsule, length=5, radius=4, shapeName=dyn_loop_FL_01_capsule, center=<3 0 4>, rotation=<rot 90 0 0>}]
 dyn_loop_FL_02 [{shapeType=Capsule, length=5, radius=5, shapeName=dyn_loop_FL_02_capsule, center=<3 0 4>, rotation=<rot 90 0 0>}]
 dyn_loop_FL_03 [{shapeType=Capsule, length=5, radius=4.5, shapeName=dyn_loop_FL_03_capsule, center=<3 0 3>, rotation=<rot 90 0 0>}]
 dyn_loop_FL_mstr [{shapeType=Capsule, length=0.505, radius=0.505, shapeName=dyn_loop_FL_mstr_capsule, rotation=<rot 0 -90 0>}]
 thigh_l [{shapeType=Capsule, length=40, radius=8, shapeName=thigh_l_capsule, center=<-20 0 0>, rotation=<rot 90 0 0>}]
 thigh_twist_01_l [{shapeType=Capsule, length=1, radius=1, shapeName=thigh_twist_01_l_capsule}]
 thigh_twist_01_r [{shapeType=Capsule, length=1, radius=1, shapeName=thigh_twist_01_r_capsule, rotation=<rot 90 0 0>}]
}
constraints [
 {
  bone1Name=calf_l
  bone2Name=thigh_l
  swing1LimitDegrees=45
  swing1Motion=Free
  swing2LimitDegrees=45
  swing2Motion=Free
  twistLimitDegrees=45
  twistMotion=Free
 }
 {bone1Name=dyn_loop_FL_03, bone2Name=dyn_loop_FL_02, swing1LimitDegrees=65, swing2LimitDegrees=65, twistLimitDegrees=5}
 {bone1Name=dyn_loop_FL_02, bone2Name=dyn_loop_FL_01, swing1LimitDegrees=45, swing2LimitDegrees=45, twistLimitDegrees=5}
 {bone1Name=dyn_loop_FL_01, bone2Name=dyn_loop_FL_mstr, swing1LimitDegrees=10, swing2LimitDegrees=45, twistLimitDegrees=2}
 {
  bone1Name=dyn_loop_FL_mstr
  bone2Name=thigh_twist_01_l
  swing1LimitDegrees=45
  swing1Motion=Free
  swing2LimitDegrees=45
  swing2Motion=Free
  twistLimitDegrees=45
  twistMotion=Free
 }
 {bone1Name=dyn_Loop_L_03, bone2Name=dyn_Loop_L_02, swing1LimitDegrees=60, swing2LimitDegrees=60, twistLimitDegrees=5}
 {bone1Name=dyn_Loop_L_02, bone2Name=dyn_Loop_L_01, swing1LimitDegrees=70, swing2LimitDegrees=70, twistLimitDegrees=5}
 {
  bone1Name=dyn_Loop_L_01
  bone2Name=dyn_Loop_L_mstr
  swing1LimitDegrees=90
  swing1Motion=Free
  swing2LimitDegrees=90
  swing2Motion=Free
  twistLimitDegrees=5
  twistMotion=Free
 }
 {
  bone1Name=dyn_Loop_L_mstr
  bone2Name=thigh_twist_01_l
  swing1LimitDegrees=45
  swing1Motion=Free
  swing2LimitDegrees=45
  swing2Motion=Free
  twistLimitDegrees=45
  twistMotion=Free
 }
 {
  bone1Name=thigh_twist_01_l
  bone2Name=thigh_l
  swing1LimitDegrees=45
  swing1Motion=Free
  swing2LimitDegrees=45
  swing2Motion=Free
  twistLimitDegrees=45
  twistMotion=Free
 }
 {
  bone1Name=dyn_loop_BL_mstr
  bone2Name=thigh_twist_01_l
  swing1LimitDegrees=45
  swing1Motion=Free
  swing2LimitDegrees=45
  swing2Motion=Free
  twistLimitDegrees=45
  twistMotion=Free
 }
 {bone1Name=dyn_loop_BL_01, bone2Name=dyn_loop_BL_mstr, swing1LimitDegrees=55, swing2LimitDegrees=55, twistLimitDegrees=45}
 {bone1Name=dyn_loop_BL_02, bone2Name=dyn_loop_BL_01, swing1LimitDegrees=55, swing2LimitDegrees=55, twistLimitDegrees=5}
 {bone1Name=dyn_loop_BL_03, bone2Name=dyn_loop_BL_02, swing1LimitDegrees=45, swing2LimitDegrees=45, twistLimitDegrees=5}
 {
  bone1Name=dyn_loop_FL_03
  bone2Name=dyn_Loop_L_02
  swing1LimitDegrees=45
  swing1Motion=Free
  swing2LimitDegrees=45
  swing2Motion=Free
  twistLimitDegrees=45
  twistMotion=Free
 }
 {
  bone1Name=dyn_loop_BL_03
  bone2Name=dyn_Loop_L_02
  swing1LimitDegrees=45
  swing1Motion=Free
  swing2LimitDegrees=45
  swing2Motion=Free
  twistLimitDegrees=45
  twistMotion=Free
 }
 {
  bone1Name=dyn_Loop_L_03
  bone2Name=dyn_loop_BL_03
  swing1LimitDegrees=45
  swing1Motion=Free
  swing2LimitDegrees=45
  swing2Motion=Free
  twistLimitDegrees=45
  twistMotion=Free
 }
 {
  bone1Name=dyn_loop_BL_03
  bone2Name=dyn_Loop_L_03
  swing1LimitDegrees=45
  swing1Motion=Free
  swing2LimitDegrees=45
  swing2Motion=Free
  twistLimitDegrees=45
  twistMotion=Free
 }
]
