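# AnimBlueprint for the female medium RainbowStraps body cosmetic, adding secondary motion to straps, zipper and flaps.
# It copies the attached parent mesh's pose, then constrains and nudges strap-loop anchor bones between pelvis and thighs.
# AnimDynamics simulates the loops in two chained passes against leg plane and sphere limits, followed by per-leg RigidBody simulations using dedicated physics assets.
# A constrained, simulated zipper and spring-driven leg and arm flaps follow.
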AnimBlueprint /BRCosmetics/Characters/Player/Female/Medium/Bodies/F_MED_RainbowStraps/Meshes/F_MED_RainbowStraps_AnimBP
target_skeleton=&/F_MED_RainbowStraps_Skeleton
graphs [
 AnimGraph {nodes [
   n1 = "Constraint - dyn_loop_FL_mstr (PPPP)"!(
    ComponentPose=Local To Component(LocalPose=Copy Pose From Mesh() {bUseAttachedParent=true})
    Alpha=1
    ConstraintWeights_0=0.05
    ConstraintWeights_1:float=0.43
    ConstraintWeights_2:float=0.49
    ConstraintWeights_3:float=0.03
   ) -> (Pose) {
    alphaBoolBlend {blendOption=Linear}
    boneToModify {boneName=dyn_loop_FL_mstr}
    constraintSetup [
     {targetBone {boneName=pelvis}}
     {targetBone {boneName=thigh_twist_01_l}}
     {targetBone {boneName=thigh_l}}
     {targetBone {boneName=deform_groin_l}}
    ]
    constraintWeights [1,1,1,1]
   }
   n2 = "Constraint - dyn_loop_FR_mstr (PPPP)"!(ComponentPose=@n1, Alpha=1, ConstraintWeights_0=0, ConstraintWeights_1:float=0.5, ConstraintWeights_2:float=0.5, ConstraintWeights_3:float=0) -> (Pose) {
    alphaBoolBlend {blendOption=Linear}
    boneToModify {boneName=dyn_loop_FR_mstr}
    constraintSetup [
     {targetBone {boneName=pelvis}}
     {targetBone {boneName=thigh_twist_01_r}}
     {targetBone {boneName=thigh_r}}
     {targetBone {boneName=deform_groin_r}}
    ]
    constraintWeights [1,1,1,1]
   }
   n3 = "Constraint - dyn_loop_BL_mstr (PP)"!(ComponentPose=@n2, Alpha=1, ConstraintWeights_0=0.5, ConstraintWeights_1:float=0.5) -> (Pose) {
    alphaBoolBlend {blendOption=Linear}
    boneToModify {boneName=dyn_loop_BL_mstr}
    constraintSetup [{targetBone {boneName=pelvis}}, {targetBone {boneName=thigh_twist_01_l}}]
    constraintWeights [1,1]
   }
   n4 = "Constraint - dyn_loop_BR_mstr (PP)"!(ComponentPose=@n3, Alpha=1, ConstraintWeights_0=0.5, ConstraintWeights_1:float=0.5) -> (Pose) {
    alphaBoolBlend {blendOption=Linear}
    boneToModify {boneName=dyn_loop_BR_mstr}
    constraintSetup [{targetBone {boneName=pelvis}}, {targetBone {boneName=thigh_twist_01_r}}]
    constraintWeights [1,1]
   }
   n5 = "Transform (Modify) Bone - Bone: dyn_loop_FL_mstr"(ComponentPose=@n4, Alpha=1, Translation="0.3,0,0.3", Rotation="0,0,0", Scale="1,1,1") {
    rotationMode=BMM_Additive
    rotationSpace=BCS_BoneSpace
    translationMode=BMM_Additive
    translationSpace=BCS_BoneSpace
    alphaBoolBlend {blendOption=Linear}
    boneToModify {boneName=dyn_loop_FL_mstr}
    translation {x=0.3, z=0.3}
   }
   n6 = "Transform (Modify) Bone - Bone: dyn_loop_FR_mstr"(ComponentPose=@n5, Alpha=1, Translation="0.3,0,0.3", Rotation="0,0,0", Scale="1,1,1") {
    rotationMode=BMM_Additive
    rotationSpace=BCS_BoneSpace
    translationMode=BMM_Additive
    translationSpace=BCS_BoneSpace
    alphaBoolBlend {blendOption=Linear}
    boneToModify {boneName=dyn_loop_FR_mstr}
    translation {x=0.3, z=0.3}
   }
   n7 = Anim Dynamics - Bone: dyn_Loop_L_01(ComponentPose=@n6, Alpha=1, ExternalForce="0,0,0") {
    angularDampingOverride=0.9
    bOverrideAngularDamping=true
    simulationSpace=RootRelative
    alphaBoolBlend {blendOption=Linear}
    boundBone {boneName=dyn_Loop_L_01}
    componentAppliedLinearAccClamp=<10000 10000 10000>
    componentLinearAccScale=<0.3 0.3 1>
    componentLinearVelScale=<0.3 0.3 1>
    physicsBodyDefinitions [{
      boundBone {boneName=dyn_Loop_L_01}
      boxExtents=<80 80 80>
      constraintSetup {angularLimitsMax=<20 5 10>, angularLimitsMin=<-20 -35 -45>, angularTarget=<0 0 0>}
      localJointOffset=<7 0 0>
     }]
    planarLimits [{drivingBone {boneName=thigh_l}, planeTransform {location=<0 0 -5>, rotation=<rot 5 0 180>}}]
   }
   n8 = Anim Dynamics - Bone: dyn_Loop_L_02(ComponentPose=@n7, Alpha=1, ExternalForce="0,0,0") {
    angularDampingOverride=0.9
    bOverrideAngularDamping=true
    simulationSpace=RootRelative
    alphaBoolBlend {blendOption=Linear}
    boundBone {boneName=dyn_Loop_L_02}
    componentAppliedLinearAccClamp=<10000 10000 10000>
    componentLinearAccScale=<0.5 0.5 1>
    componentLinearVelScale=<0.5 0.5 1>
    physicsBodyDefinitions [{
      boundBone {boneName=dyn_Loop_L_02}
      boxExtents=<40 40 40>
      constraintSetup {angularLimitsMax=<20 5 10>, angularLimitsMin=<-20 -35 -45>, angularTarget=<0 0 0>}
      localJointOffset=<5 0 0>
     }]
    planarLimits [{drivingBone {boneName=thigh_l}, planeTransform {location=<0 0 -5>, rotation=<rot 5 0 -180>}}]
    sphericalLimits [
     {limitRadius=13, drivingBone {boneName=thigh_l}, sphereLocalOffset=<-30 0 0>}
     {limitRadius=17, drivingBone {boneName=calf_l}, sphereLocalOffset=<0 0 0>}
     {limitRadius=18, drivingBone {boneName=calf_l}, sphereLocalOffset=<-27 0 0>}
    ]
   }
   n9 = Anim Dynamics - Bone: dyn_Loop_L_03(ComponentPose=@n8, Alpha=1, ExternalForce="0,0,0") {
    angularDampingOverride=0.9
    bOverrideAngularDamping=true
    simulationSpace=RootRelative
    alphaBoolBlend {blendOption=Linear}
    boundBone {boneName=dyn_Loop_L_03}
    componentAppliedLinearAccClamp=<10000 10000 10000>
    componentLinearAccScale=<1 1 2>
    componentLinearVelScale=<1 1 2>
    physicsBodyDefinitions [{
      boundBone {boneName=dyn_Loop_L_03}
      boxExtents=<20 20 20>
      constraintSetup {angularLimitsMax=<20 20 30>, angularLimitsMin=<-20 -55 -30>, angularTarget=<0 0 0>}
      localJointOffset=<3 0 0>
     }]
    planarLimits [{drivingBone {boneName=thigh_l}, planeTransform {location=<0 0 -5>, rotation=<rot 5 0 -180>}}]
    sphericalLimits [
     {limitRadius=16, drivingBone {boneName=calf_l}, sphereLocalOffset=<0 0 0>}
     {limitRadius=25, drivingBone {boneName=calf_l}, sphereLocalOffset=<-20 0 0>}
    ]
   }
   n10 = Anim Dynamics - Bone: dyn_loop_BL_02(ComponentPose=@n9, Alpha=1, ExternalForce="0,0,0") {
    angularDampingOverride=0.99
    bOverrideAngularDamping=true
    bUseSphericalLimits=true
    simulationSpace=RootRelative
    alphaBoolBlend {blendOption=Linear}
    boundBone {boneName=dyn_loop_BL_02}
    componentAppliedLinearAccClamp=<10000 10000 10000>
    componentLinearAccScale=<0.5 0.5 1>
    componentLinearVelScale=<0.5 0.5 1>
    physicsBodyDefinitions [{
      boundBone {boneName=dyn_loop_BL_02}
      boxExtents=<60 60 60>
      constraintSetup {angularLimitsMax=<20 10 20>, angularLimitsMin=<-20 -55 -20>, angularTarget=<0 0 0>}
      localJointOffset=<5 0 0>
     }]
    planarLimits [
     {drivingBone {boneName=thigh_l}, planeTransform {location=<0 0 -10>, rotation=<rot 5 0.0001 -115>}}
     {drivingBone {boneName=calf_l}, planeTransform {location=<0 5 0>, rotation=<rot 0 0 -90>}}
    ]
    sphericalLimits [
     {limitRadius=10, drivingBone {boneName=thigh_l}, sphereLocalOffset=<-35 0 0>}
     {limitRadius=10, drivingBone {boneName=calf_l}, sphereLocalOffset=<0 0 0>}
     {limitRadius=25, drivingBone {boneName=calf_l}, sphereLocalOffset=<-30 0 0>}
    ]
   }
   n11 = Anim Dynamics - Bone: dyn_loop_BL_03(ComponentPose=@n10, Alpha=1, ExternalForce="0,0,0") {
    angularDampingOverride=0.99
    bOverrideAngularDamping=true
    simulationSpace=RootRelative
    alphaBoolBlend {blendOption=Linear}
    boundBone {boneName=dyn_loop_BL_03}
    componentAppliedLinearAccClamp=<10000 10000 10000>
    componentLinearAccScale=<1 1 2>
    componentLinearVelScale=<1 1 2>
    physicsBodyDefinitions [{
      boundBone {boneName=dyn_loop_BL_03}
      boxExtents=<60 60 60>
      constraintSetup {angularLimitsMax=<20 35 30>, angularLimitsMin=<-20 -55 -30>, angularTarget=<0 0 0>}
      localJointOffset=<5 0 0>
     }]
    planarLimits [
     {drivingBone {boneName=thigh_l}, planeTransform {location=<0 5 5>, rotation=<rot 5 -0.0001 -115>}}
     {drivingBone {boneName=calf_l}, planeTransform {location=<0 5 0>, rotation=<rot 0 0 -90>}}
    ]
    sphericalLimits [
     {limitRadius=15, drivingBone {boneName=thigh_l}, sphereLocalOffset=<-30 0 0>}
     {limitRadius=20, drivingBone {boneName=calf_l}, sphereLocalOffset=<0 0 0>}
     {limitRadius=25, drivingBone {boneName=calf_l}, sphereLocalOffset=<-20 0 0>}
    ]
   }
   n12 = Anim Dynamics - Bone: dyn_loop_R_01(ComponentPose=@n11, Alpha=1, ExternalForce="0,0,0") {
    angularDampingOverride=0.9
    bOverrideAngularDamping=true
    simulationSpace=RootRelative
    alphaBoolBlend {blendOption=Linear}
    boundBone {boneName=dyn_loop_R_01}
    componentAppliedLinearAccClamp=<10000 10000 10000>
    componentLinearAccScale=<0.3 0.3 1>
    componentLinearVelScale=<0.3 0.3 1>
    physicsBodyDefinitions [{
      boundBone {boneName=dyn_loop_R_01}
      boxExtents=<80 80 80>
      constraintSetup {angularLimitsMax=<20 5 10>, angularLimitsMin=<-20 -35 -10>, angularTarget=<0 0 0>}
      localJointOffset=<7 0 0>
     }]
    planarLimits [
     {drivingBone {boneName=thigh_r}, planeTransform {location=<0 0 -5>, rotation=<rot 5 0 25>}}
     {drivingBone {boneName=calf_r}, planeTransform {location=<0 5 0>, rotation=<rot 0 0 25>}}
    ]
   }
   n13 = Anim Dynamics - Bone: dyn_loop_R_02(ComponentPose=@n12, Alpha=1, ExternalForce="0,0,0") {
    angularDampingOverride=0.9
    bOverrideAngularDamping=true
    bUseSphericalLimits=true
    simulationSpace=RootRelative
    alphaBoolBlend {blendOption=Linear}
    boundBone {boneName=dyn_loop_R_02}
    componentAppliedLinearAccClamp=<10000 10000 10000>
    componentLinearAccScale=<0.3 0.3 1>
    componentLinearVelScale=<0.3 0.3 1>
    physicsBodyDefinitions [{
      boundBone {boneName=dyn_loop_R_02}
      boxExtents=<40 40 40>
      constraintSetup {angularLimitsMax=<20 5 10>, angularLimitsMin=<-20 -35 -10>, angularTarget=<0 0 0>}
      localJointOffset=<5 0 0>
     }]
    planarLimits [
     {drivingBone {boneName=thigh_r}, planeTransform {location=<0 0 -5>, rotation=<rot 5 0 25>}}
     {drivingBone {boneName=calf_r}, planeTransform {location=<0 5 0>, rotation=<rot 0 0 25>}}
    ]
    sphericalLimits [
     {limitRadius=12, drivingBone {boneName=thigh_r}, sphereLocalOffset=<35 0 0>}
     {limitRadius=15, drivingBone {boneName=calf_r}, sphereLocalOffset=<0 0 0>}
     {limitRadius=18, drivingBone {boneName=calf_r}, sphereLocalOffset=<27 0 0>}
    ]
   }
   n14 = Anim Dynamics - Bone: dyn_loop_R_03(ComponentPose=@n13, Alpha=1, ExternalForce="0,0,0") {
    angularDampingOverride=0.9
    bOverrideAngularDamping=true
    simulationSpace=RootRelative
    alphaBoolBlend {blendOption=Linear}
    boundBone {boneName=dyn_loop_R_03}
    componentAppliedLinearAccClamp=<10000 10000 10000>
    componentLinearAccScale=<0.5 0.5 2>
    componentLinearVelScale=<0.5 0.5 2>
    physicsBodyDefinitions [{
      boundBone {boneName=dyn_loop_R_03}
      boxExtents=<20 20 20>
      constraintSetup {angularLimitsMax=<20 20 30>, angularLimitsMin=<-20 -55 -30>, angularTarget=<0 0 0>}
      localJointOffset=<3 0 0>
     }]
    planarLimits [
     {drivingBone {boneName=thigh_r}, planeTransform {location=<0 0 -5>, rotation=<rot 5 0 25>}}
     {drivingBone {boneName=calf_r}, planeTransform {location=<0 5 0>, rotation=<rot 0 0 25>}}
    ]
    sphericalLimits [
     {limitRadius=12, drivingBone {boneName=thigh_r}, sphereLocalOffset=<40 0 0>}
     {limitRadius=15, drivingBone {boneName=calf_r}, sphereLocalOffset=<0 0 0>}
     {limitRadius=18, drivingBone {boneName=calf_r}, sphereLocalOffset=<20 0 0>}
    ]
   }
   n15 = Anim Dynamics - Bone: dyn_loop_BR_02(ComponentPose=@n14, Alpha=1, ExternalForce="0,0,0") {
    angularDampingOverride=0.99
    bOverrideAngularDamping=true
    bUseSphericalLimits=true
    simulationSpace=RootRelative
    alphaBoolBlend {blendOption=Linear}
    boundBone {boneName=dyn_loop_BR_02}
    componentAppliedLinearAccClamp=<10000 10000 10000>
    componentLinearAccScale=<0.5 0.5 1>
    componentLinearVelScale=<0.5 0.5 1>
    physicsBodyDefinitions [{
      boundBone {boneName=dyn_loop_BR_02}
      boxExtents=<60 60 60>
      constraintSetup {angularLimitsMax=<20 5 20>, angularLimitsMin=<-20 -55 -20>, angularTarget=<0 0 0>}
      localJointOffset=<5 0 0>
     }]
    planarLimits [
     {drivingBone {boneName=thigh_r}, planeTransform {location=<0 6 5>, rotation=<rot 5 0 85>}}
     {drivingBone {boneName=calf_r}, planeTransform {location=<0 5 0>, rotation=<rot 0 0 25>}}
    ]
    sphericalLimits [
     {limitRadius=15, drivingBone {boneName=thigh_r}, sphereLocalOffset=<35 0 0>}
     {limitRadius=20, drivingBone {boneName=calf_r}, sphereLocalOffset=<0 0 0>}
     {limitRadius=30, drivingBone {boneName=calf_r}, sphereLocalOffset=<20 0 0>}
    ]
   }
   n16 = Anim Dynamics - Bone: dyn_loop_BR_03(ComponentPose=@n15, Alpha=1, ExternalForce="0,0,0") {
    angularDampingOverride=0.99
    bOverrideAngularDamping=true
    simulationSpace=RootRelative
    alphaBoolBlend {blendOption=Linear}
    boundBone {boneName=dyn_loop_BR_03}
    componentAppliedLinearAccClamp=<10000 10000 10000>
    componentLinearAccScale=<1 1 2>
    componentLinearVelScale=<1 1 2>
    physicsBodyDefinitions [{
      boundBone {boneName=dyn_loop_BR_03}
      boxExtents=<60 60 60>
      constraintSetup {angularLimitsMax=<20 5 30>, angularLimitsMin=<-20 -55 -30>, angularTarget=<0 0 0>}
      localJointOffset=<5 0 0>
     }]
    planarLimits [
     {drivingBone {boneName=thigh_r}, planeTransform {location=<0 5 5>, rotation=<rot 5 0 85>}}
     {drivingBone {boneName=calf_r}, planeTransform {location=<0 5 0>, rotation=<rot 0 0 25>}}
    ]
    sphericalLimits [
     {limitRadius=10, drivingBone {boneName=calf_r}, sphereLocalOffset=<0 0 0>}
     {limitRadius=20, drivingBone {boneName=calf_r}, sphereLocalOffset=<0 0 0>}
     {limitRadius=25, drivingBone {boneName=calf_r}, sphereLocalOffset=<20 0 0>}
    ]
   }
   n17 = Anim Dynamics - Bone: dyn_Loop_L_01(ComponentPose=@n16, Alpha=1, ExternalForce="0,0,0") {
    angularDampingOverride=0.9
    bOverrideAngularDamping=true
    simulationSpace=RootRelative
    alphaBoolBlend {blendOption=Linear}
    boundBone {boneName=dyn_Loop_L_01}
    componentAppliedLinearAccClamp=<10000 10000 10000>
    componentLinearAccScale=<0.3 0.3 1>
    componentLinearVelScale=<0.3 0.3 1>
    physicsBodyDefinitions [{
      boundBone {boneName=dyn_Loop_L_01}
      boxExtents=<80 80 80>
      constraintSetup {angularLimitsMax=<20 5 10>, angularLimitsMin=<-20 -35 -45>, angularTarget=<0 0 0>}
      localJointOffset=<7 0 0>
     }]
    planarLimits [{drivingBone {boneName=thigh_l}, planeTransform {location=<0 0 -5>, rotation=<rot 5 0 -180>}}]
   }
   n18 = Anim Dynamics - Bone: dyn_Loop_L_02(ComponentPose=@n17, Alpha=1, ExternalForce="0,0,0") {
    angularDampingOverride=0.9
    bOverrideAngularDamping=true
    simulationSpace=RootRelative
    alphaBoolBlend {blendOption=Linear}
    boundBone {boneName=dyn_Loop_L_02}
    componentAppliedLinearAccClamp=<10000 10000 10000>
    componentLinearAccScale=<0.5 0.5 1>
    componentLinearVelScale=<0.5 0.5 1>
    physicsBodyDefinitions [{
      boundBone {boneName=dyn_Loop_L_02}
      boxExtents=<40 40 40>
      constraintSetup {angularLimitsMax=<20 5 10>, angularLimitsMin=<-20 -35 -45>, angularTarget=<0 0 0>}
      localJointOffset=<5 0 0>
     }]
    planarLimits [{drivingBone {boneName=thigh_l}, planeTransform {location=<0 0 -5>, rotation=<rot 5 0 -180>}}]
    sphericalLimits [
     {limitRadius=12, drivingBone {boneName=thigh_l}, sphereLocalOffset=<-30 0 0>}
     {limitRadius=12, drivingBone {boneName=calf_l}, sphereLocalOffset=<0 0 0>}
     {limitRadius=18, drivingBone {boneName=calf_l}, sphereLocalOffset=<-27 0 0>}
    ]
   }
   n19 = Anim Dynamics - Bone: dyn_Loop_L_03(ComponentPose=@n18, Alpha=1, ExternalForce="0,0,0") {
    angularDampingOverride=0.9
    bOverrideAngularDamping=true
    simulationSpace=RootRelative
    alphaBoolBlend {blendOption=Linear}
    boundBone {boneName=dyn_Loop_L_03}
    componentAppliedLinearAccClamp=<10000 10000 10000>
    componentLinearAccScale=<1 1 2>
    componentLinearVelScale=<1 1 2>
    physicsBodyDefinitions [{
      boundBone {boneName=dyn_Loop_L_03}
      boxExtents=<20 20 20>
      constraintSetup {angularLimitsMax=<20 20 30>, angularLimitsMin=<-20 -55 -30>, angularTarget=<0 0 0>}
      localJointOffset=<3 0 0>
     }]
    planarLimits [{drivingBone {boneName=thigh_l}, planeTransform {location=<0 0 -5>, rotation=<rot 5 0 -180>}}]
    sphericalLimits [
     {limitRadius=10, drivingBone {boneName=calf_l}, sphereLocalOffset=<0 0 0>}
     {limitRadius=10, drivingBone {boneName=calf_l}, sphereLocalOffset=<-20 0 0>}
    ]
   }
   n20 = Anim Dynamics - Bone: dyn_loop_BL_02(ComponentPose=@n19, Alpha=1, ExternalForce="0,0,0") {
    angularDampingOverride=0.99
    bOverrideAngularDamping=true
    bUseSphericalLimits=true
    simulationSpace=RootRelative
    alphaBoolBlend {blendOption=Linear}
    boundBone {boneName=dyn_loop_BL_02}
    componentAppliedLinearAccClamp=<10000 10000 10000>
    componentLinearAccScale=<0.5 0.5 1>
    componentLinearVelScale=<0.5 0.5 1>
    physicsBodyDefinitions [{
      boundBone {boneName=dyn_loop_BL_02}
      boxExtents=<60 60 60>
      constraintSetup {angularLimitsMax=<20 10 20>, angularLimitsMin=<-20 -55 -20>, angularTarget=<0 0 0>}
      localJointOffset=<5 0 0>
     }]
    planarLimits [
     {drivingBone {boneName=thigh_l}, planeTransform {location=<0 0 -10>, rotation=<rot 4.9999 0.0001 -115>}}
     {drivingBone {boneName=calf_l}, planeTransform {location=<0 5 0>, rotation=<rot 0 0 -90>}}
    ]
    sphericalLimits [
     {limitRadius=10, drivingBone {boneName=thigh_l}, sphereLocalOffset=<-35 0 0>}
     {limitRadius=10, drivingBone {boneName=calf_l}, sphereLocalOffset=<0 0 0>}
     {limitRadius=15, drivingBone {boneName=calf_l}, sphereLocalOffset=<-30 0 0>}
    ]
   }
   n21 = Anim Dynamics - Bone: dyn_loop_BL_03(ComponentPose=@n20, Alpha=1, ExternalForce="0,0,0") {
    angularDampingOverride=0.99
    bOverrideAngularDamping=true
    simulationSpace=RootRelative
    alphaBoolBlend {blendOption=Linear}
    boundBone {boneName=dyn_loop_BL_03}
    componentAppliedLinearAccClamp=<10000 10000 10000>
    componentLinearAccScale=<1 1 2>
    componentLinearVelScale=<1 1 2>
    physicsBodyDefinitions [{
      boundBone {boneName=dyn_loop_BL_03}
      boxExtents=<60 60 60>
      constraintSetup {angularLimitsMax=<20 35 30>, angularLimitsMin=<-20 -55 -30>, angularTarget=<0 0 0>}
      localJointOffset=<5 0 0>
     }]
    planarLimits [
     {drivingBone {boneName=thigh_l}, planeTransform {location=<0 5 5>, rotation=<rot 5 0 -115.0001>}}
     {drivingBone {boneName=calf_l}, planeTransform {location=<0 5 0>, rotation=<rot 0 0 -90>}}
    ]
    sphericalLimits [
     {limitRadius=10, drivingBone {boneName=thigh_l}, sphereLocalOffset=<-30 0 0>}
     {limitRadius=10, drivingBone {boneName=calf_l}, sphereLocalOffset=<0 0 0>}
     {limitRadius=15, drivingBone {boneName=calf_l}, sphereLocalOffset=<-20 0 0>}
    ]
   }
   n22 = Anim Dynamics - Bone: dyn_loop_R_01(ComponentPose=@n21, Alpha=1, ExternalForce="0,0,0") {
    angularDampingOverride=0.9
    bOverrideAngularDamping=true
    simulationSpace=RootRelative
    alphaBoolBlend {blendOption=Linear}
    boundBone {boneName=dyn_loop_R_01}
    componentAppliedLinearAccClamp=<10000 10000 10000>
    componentLinearAccScale=<0.3 0.3 1>
    componentLinearVelScale=<0.3 0.3 1>
    physicsBodyDefinitions [{
      boundBone {boneName=dyn_loop_R_01}
      boxExtents=<80 80 80>
      constraintSetup {angularLimitsMax=<20 5 10>, angularLimitsMin=<-20 -35 -10>, angularTarget=<0 0 0>}
      localJointOffset=<7 0 0>
     }]
    planarLimits [
     {drivingBone {boneName=thigh_r}, planeTransform {location=<0 0 -5>, rotation=<rot 5 0 25.0001>}}
     {drivingBone {boneName=calf_r}, planeTransform {location=<0 5 0>, rotation=<rot 0 0 25>}}
    ]
   }
   n23 = Anim Dynamics - Bone: dyn_loop_R_02(ComponentPose=@n22, Alpha=1, ExternalForce="0,0,0") {
    angularDampingOverride=0.9
    bOverrideAngularDamping=true
    bUseSphericalLimits=true
    simulationSpace=RootRelative
    alphaBoolBlend {blendOption=Linear}
    boundBone {boneName=dyn_loop_R_02}
    componentAppliedLinearAccClamp=<10000 10000 10000>
    componentLinearAccScale=<0.3 0.3 1>
    componentLinearVelScale=<0.3 0.3 1>
    physicsBodyDefinitions [{
      boundBone {boneName=dyn_loop_R_02}
      boxExtents=<40 40 40>
      constraintSetup {angularLimitsMax=<20 5 10>, angularLimitsMin=<-20 -35 -10>, angularTarget=<0 0 0>}
      localJointOffset=<5 0 0>
     }]
    planarLimits [
     {drivingBone {boneName=thigh_r}, planeTransform {location=<0 0 -5>, rotation=<rot 5 0 25.0001>}}
     {drivingBone {boneName=calf_r}, planeTransform {location=<0 5 0>, rotation=<rot 0 0 25>}}
    ]
    sphericalLimits [
     {limitRadius=12, drivingBone {boneName=thigh_r}, sphereLocalOffset=<35 0 0>}
     {limitRadius=12, drivingBone {boneName=calf_r}, sphereLocalOffset=<0 0 0>}
     {limitRadius=15, drivingBone {boneName=calf_r}, sphereLocalOffset=<27 0 0>}
    ]
   }
   n24 = Anim Dynamics - Bone: dyn_loop_R_03(ComponentPose=@n23, Alpha=1, ExternalForce="0,0,0") {
    angularDampingOverride=0.9
    bOverrideAngularDamping=true
    bUseSphericalLimits=true
    simulationSpace=RootRelative
    alphaBoolBlend {blendOption=Linear}
    boundBone {boneName=dyn_loop_R_03}
    componentAppliedLinearAccClamp=<10000 10000 10000>
    componentLinearAccScale=<0.5 0.5 2>
    componentLinearVelScale=<0.5 0.5 2>
    physicsBodyDefinitions [{
      boundBone {boneName=dyn_loop_R_03}
      boxExtents=<20 20 20>
      constraintSetup {angularLimitsMax=<20 20 30>, angularLimitsMin=<-20 -55 -30>, angularTarget=<0 0 0>}
      localJointOffset=<3 0 0>
     }]
    planarLimits [
     {drivingBone {boneName=thigh_r}, planeTransform {location=<0 0 -5>, rotation=<rot 5 0 25.0001>}}
     {drivingBone {boneName=calf_r}, planeTransform {location=<0 5 0>, rotation=<rot 0 0 25>}}
    ]
    sphericalLimits [
     {limitRadius=10, drivingBone {boneName=thigh_r}, sphereLocalOffset=<40 0 0>}
     {limitRadius=10, drivingBone {boneName=calf_r}, sphereLocalOffset=<0 0 0>}
     {limitRadius=10, drivingBone {boneName=calf_r}, sphereLocalOffset=<20 0 0>}
    ]
   }
   n25 = Anim Dynamics - Bone: dyn_loop_BR_02(ComponentPose=@n24, Alpha=1, ExternalForce="0,0,0") {
    angularDampingOverride=0.99
    bOverrideAngularDamping=true
    bUseSphericalLimits=true
    simulationSpace=RootRelative
    alphaBoolBlend {blendOption=Linear}
    boundBone {boneName=dyn_loop_BR_02}
    componentAppliedLinearAccClamp=<10000 10000 10000>
    componentLinearAccScale=<0.5 0.5 1>
    componentLinearVelScale=<0.5 0.5 1>
    physicsBodyDefinitions [{
      boundBone {boneName=dyn_loop_BR_02}
      boxExtents=<60 60 60>
      constraintSetup {angularLimitsMax=<20 5 20>, angularLimitsMin=<-20 -55 -20>, angularTarget=<0 0 0>}
      localJointOffset=<5 0 0>
     }]
    planarLimits [
     {drivingBone {boneName=thigh_r}, planeTransform {location=<0 6 5>, rotation=<rot 5 0.0001 85>}}
     {drivingBone {boneName=calf_r}, planeTransform {location=<0 5 0>, rotation=<rot 0 0 25>}}
    ]
    sphericalLimits [
     {limitRadius=10, drivingBone {boneName=thigh_r}, sphereLocalOffset=<35 0 0>}
     {limitRadius=10, drivingBone {boneName=calf_r}, sphereLocalOffset=<0 0 0>}
     {limitRadius=15, drivingBone {boneName=calf_r}, sphereLocalOffset=<20 0 0>}
    ]
   }
   n26 = Anim Dynamics - Bone: dyn_loop_BR_03(ComponentPose=@n25, Alpha=1, ExternalForce="0,0,0") {
    angularDampingOverride=0.99
    bOverrideAngularDamping=true
    simulationSpace=RootRelative
    alphaBoolBlend {blendOption=Linear}
    boundBone {boneName=dyn_loop_BR_03}
    componentAppliedLinearAccClamp=<10000 10000 10000>
    componentLinearAccScale=<1 1 2>
    componentLinearVelScale=<1 1 2>
    physicsBodyDefinitions [{
      boundBone {boneName=dyn_loop_BR_03}
      boxExtents=<60 60 60>
      constraintSetup {angularLimitsMax=<20 5 30>, angularLimitsMin=<-20 -55 -30>, angularTarget=<0 0 0>}
      localJointOffset=<5 0 0>
     }]
    planarLimits [
     {drivingBone {boneName=thigh_r}, planeTransform {location=<0 5 5>, rotation=<rot 5 -0.0001 85>}}
     {drivingBone {boneName=calf_r}, planeTransform {location=<0 5 0>, rotation=<rot 0 0 25>}}
    ]
    sphericalLimits [
     {limitRadius=10, drivingBone {boneName=calf_r}, sphereLocalOffset=<0 0 0>}
     {limitRadius=10, drivingBone {boneName=calf_r}, sphereLocalOffset=<0 0 0>}
     {limitRadius=15, drivingBone {boneName=calf_r}, sphereLocalOffset=<20 0 0>}
    ]
   }
   n27 = RigidBody!(ComponentPose=@n26, Alpha, OverrideWorldGravity:Vector="0,0,0", ExternalForce) -> (Pose) {
    bClampLinearTranslationLimitToRefPose=true
    bOverrideWorldGravity=true
    overridePhysicsAsset=&/F_MED_RainbowStraps_Straps
    simulationSpace=BaseBoneSpace
    alphaBoolBlend {blendOption=Linear}
    baseBoneRef {boneName=thigh_l}
    componentAppliedLinearAccClamp=<3000 3000 3000>
    componentLinearAccScale=<0.5 0.5 0.5>
    componentLinearVelScale=<0.5 0.5 0.5>
    simSpaceSettings {
     maxAngularAcceleration=1000
     maxAngularVelocity=3
     maxLinearAcceleration=700
     maxLinearVelocity=500
     velocityScaleZ=0.2
     worldAlpha=0.5
     externalLinearDragV=<0.2 0.2>
    }
   }
   n28 = RigidBody(ComponentPose=@n27, OverrideWorldGravity="0,0,0") {
    bClampLinearTranslationLimitToRefPose=true
    bOverrideWorldGravity=true
    overridePhysicsAsset=&/F_MED_RainbowStraps_Straps_Rightleg
    simulationSpace=BaseBoneSpace
    alphaBoolBlend {blendOption=Linear}
    baseBoneRef {boneName=thigh_r}
    componentAppliedLinearAccClamp=<3000 3000 3000>
    componentLinearAccScale=<0.5 0.5 0.5>
    componentLinearVelScale=<0.5 0.5 0.5>
    simSpaceSettings {
     maxAngularAcceleration=1000
     maxAngularVelocity=3
     maxLinearAcceleration=700
     maxLinearVelocity=500
     velocityScaleZ=0.2
     worldAlpha=0.5
     externalLinearDragV=<0.2 0.2>
    }
   }
   n29 = "Constraint - dyn_zipper_mstr (PPP)"!(ComponentPose=@n28, Alpha=1, ConstraintWeights_0=0.2, ConstraintWeights_1:float=0.4, ConstraintWeights_2:float=0.4) -> (Pose) {
    alphaBoolBlend {blendOption=Linear}
    boneToModify {boneName=dyn_zipper_mstr}
    constraintSetup [{targetBone {boneName=spine_05}}, {targetBone {boneName=neck_01}}, {targetBone {boneName=neck_02}}]
    constraintWeights [1,1,1]
   }
   n30 = Anim Dynamics - Bone: dyn_zipper(ComponentPose=@n29, Alpha=1) {
    angularDampingOverride=0.9
    bOverrideAngularDamping=true
    alphaBoolBlend {blendOption=Linear}
    boundBone {boneName=dyn_zipper}
    componentAppliedLinearAccClamp=<10000 10000 10000>
    physicsBodyDefinitions [{
      boundBone {boneName=dyn_zipper}
      boxExtents=<15 15 15>
      constraintSetup {angularLimitsMax=<0 0 15>, angularLimitsMin=<0 -35 -15>, angularTarget=<0 0 0>}
      localJointOffset=<1 0 0>
     }]
   }
   n31 = Anim Dynamics - Bone: dyn_leg_flap_01_L(ComponentPose=@n30, Alpha=1, ExternalForce="0,0,0") {
    angularDampingOverride=0.9999
    angularSpringConstant=5000
    bAngularSpring=true
    bOverrideAngularDamping=true
    simulationSpace=RootRelative
    alphaBoolBlend {blendOption=Linear}
    boundBone {boneName=dyn_leg_flap_01_L}
    componentAppliedLinearAccClamp=<10000 10000 10000>
    componentLinearAccScale=<3 3 3>
    componentLinearVelScale=<3 3 3>
    physicsBodyDefinitions [{
      boundBone {boneName=dyn_leg_flap_01_L}
      boxExtents=<20 10 10>
      constraintSetup {angularLimitsMax=<0 0 20>, angularLimitsMin=<0 0 -3>, angularTarget=<0 0 0>}
      localJointOffset=<4 0 0>
     }]
    relativeSpaceBone {boneName=pelvis}
   }
   n32 = Anim Dynamics - Bone: dyn_leg_flap_03_L(ComponentPose=@n31, Alpha=1, ExternalForce="0,0,0") {
    angularDampingOverride=0.9999
    angularSpringConstant=5000
    bAngularSpring=true
    bOverrideAngularDamping=true
    simulationSpace=RootRelative
    alphaBoolBlend {blendOption=Linear}
    boundBone {boneName=dyn_leg_flap_03_L}
    componentAppliedLinearAccClamp=<10000 10000 10000>
    componentLinearAccScale=<3 3 3>
    componentLinearVelScale=<3 3 3>
    physicsBodyDefinitions [{
      boundBone {boneName=dyn_leg_flap_03_L}
      boxExtents=<20 10 10>
      constraintSetup {angularLimitsMax=<0 0 20>, angularLimitsMin=<0 0 -3>, angularTarget=<0 0 0>}
      localJointOffset=<4 0 0>
     }]
    relativeSpaceBone {boneName=pelvis}
   }
   n33 = Anim Dynamics - Bone: dyn_leg_flap_05_L(ComponentPose=@n32, Alpha=1, ExternalForce="0,0,0") {
    angularDampingOverride=0.9999
    angularSpringConstant=5000
    bAngularSpring=true
    bOverrideAngularDamping=true
    simulationSpace=RootRelative
    alphaBoolBlend {blendOption=Linear}
    boundBone {boneName=dyn_leg_flap_05_L}
    componentAppliedLinearAccClamp=<10000 10000 10000>
    componentLinearAccScale=<3 3 3>
    componentLinearVelScale=<3 3 3>
    physicsBodyDefinitions [{
      boundBone {boneName=dyn_leg_flap_05_L}
      boxExtents=<20 10 10>
      constraintSetup {angularLimitsMax=<0 0 20>, angularLimitsMin=<0 0 -3>, angularTarget=<0 0 0>}
      localJointOffset=<4 0 0>
     }]
    relativeSpaceBone {boneName=pelvis}
   }
   n34 = Anim Dynamics - Bone: dyn_leg_flap_02_L(ComponentPose=@n33, Alpha=1, ExternalForce="0,0,0") {
    angularDampingOverride=0.9999
    angularSpringConstant=5000
    bAngularSpring=true
    bOverrideAngularDamping=true
    simulationSpace=RootRelative
    alphaBoolBlend {blendOption=Linear}
    boundBone {boneName=dyn_leg_flap_02_L}
    componentAppliedLinearAccClamp=<10000 10000 10000>
    componentLinearAccScale=<3 3 3>
    componentLinearVelScale=<3 3 3>
    physicsBodyDefinitions [{
      boundBone {boneName=dyn_leg_flap_02_L}
      boxExtents=<20 10 10>
      constraintSetup {angularLimitsMax=<0 0 3>, angularLimitsMin=<0 0 -20>, angularTarget=<0 0 0>}
      localJointOffset=<4 0 0>
     }]
    relativeSpaceBone {boneName=pelvis}
   }
   n35 = Anim Dynamics - Bone: dyn_leg_flap_04_L(ComponentPose=@n34, Alpha=1, ExternalForce="0,0,0") {
    angularDampingOverride=0.9999
    angularSpringConstant=5000
    bAngularSpring=true
    bOverrideAngularDamping=true
    simulationSpace=RootRelative
    alphaBoolBlend {blendOption=Linear}
    boundBone {boneName=dyn_leg_flap_04_L}
    componentAppliedLinearAccClamp=<10000 10000 10000>
    componentLinearAccScale=<3 3 3>
    componentLinearVelScale=<3 3 3>
    physicsBodyDefinitions [{
      boundBone {boneName=dyn_leg_flap_04_L}
      boxExtents=<20 10 10>
      constraintSetup {angularLimitsMax=<0 0 3>, angularLimitsMin=<0 0 -20>, angularTarget=<0 0 0>}
      localJointOffset=<4 0 0>
     }]
    relativeSpaceBone {boneName=pelvis}
   }
   n36 = Anim Dynamics - Bone: dyn_leg_flap_06_L(ComponentPose=@n35, Alpha=1, ExternalForce="0,0,0") {
    angularDampingOverride=0.9999
    angularSpringConstant=5000
    bAngularSpring=true
    bOverrideAngularDamping=true
    simulationSpace=RootRelative
    alphaBoolBlend {blendOption=Linear}
    boundBone {boneName=dyn_leg_flap_06_L}
    componentAppliedLinearAccClamp=<10000 10000 10000>
    componentLinearAccScale=<3 3 3>
    componentLinearVelScale=<3 3 3>
    physicsBodyDefinitions [{
      boundBone {boneName=dyn_leg_flap_06_L}
      boxExtents=<20 10 10>
      constraintSetup {angularLimitsMax=<0 0 3>, angularLimitsMin=<0 0 -20>, angularTarget=<0 0 0>}
      localJointOffset=<4 0 0>
     }]
    relativeSpaceBone {boneName=pelvis}
   }
   n37 = Anim Dynamics - Bone: dyn_arm_flap_01_L(ComponentPose=@n36, Alpha=1, ExternalForce="0,0,0") {
    angularDampingOverride=0.9999
    angularSpringConstant=50000
    bAngularSpring=true
    bOverrideAngularDamping=true
    simulationSpace=RootRelative
    alphaBoolBlend {blendOption=Linear}
    boundBone {boneName=dyn_arm_flap_01_L}
    componentAppliedLinearAccClamp=<10000 10000 10000>
    componentLinearAccScale=<3 3 3>
    componentLinearVelScale=<3 3 3>
    physicsBodyDefinitions [{
      boundBone {boneName=dyn_arm_flap_01_L}
      boxExtents=<15 15 15>
      constraintSetup {angularLimitsMax=<1 10 10>, angularLimitsMin=<-1 -10 -10>, angularTarget=<0 0 0>}
      localJointOffset=<2 0 0>
     }]
    relativeSpaceBone {boneName=spine_05}
   }
   n38 = Anim Dynamics - Bone: dyn_arm_flap_02_L(ComponentPose=@n37, Alpha=1, ExternalForce="0,0,0") {
    angularDampingOverride=0.9999
    angularSpringConstant=50000
    bAngularSpring=true
    bOverrideAngularDamping=true
    simulationSpace=RootRelative
    alphaBoolBlend {blendOption=Linear}
    boundBone {boneName=dyn_arm_flap_02_L}
    componentAppliedLinearAccClamp=<10000 10000 10000>
    componentLinearAccScale=<3 3 3>
    componentLinearVelScale=<3 3 3>
    physicsBodyDefinitions [{
      boundBone {boneName=dyn_arm_flap_02_L}
      boxExtents=<7 7 7>
      constraintSetup {angularLimitsMax=<1 10 10>, angularLimitsMin=<-1 -10 -10>, angularTarget=<0 0 0>}
      localJointOffset=<2 0 0>
     }]
    relativeSpaceBone {boneName=spine_05}
   }
   n39 = Anim Dynamics - Bone: dyn_arm_flap_03_L(ComponentPose=@n38, Alpha=1, ExternalForce="0,0,0") {
    angularDampingOverride=0.9999
    angularSpringConstant=40000
    bAngularSpring=true
    bOverrideAngularDamping=true
    simulationSpace=RootRelative
    alphaBoolBlend {blendOption=Linear}
    boundBone {boneName=dyn_arm_flap_03_L}
    componentAppliedLinearAccClamp=<10000 10000 10000>
    componentLinearAccScale=<3 3 3>
    componentLinearVelScale=<3 3 3>
    physicsBodyDefinitions [{
      boundBone {boneName=dyn_arm_flap_03_L}
      boxExtents=<10 10 10>
      constraintSetup {angularLimitsMax=<1 10 10>, angularLimitsMin=<-1 -10 -10>, angularTarget=<0 0 0>}
      localJointOffset=<1 0 0>
     }]
    relativeSpaceBone {boneName=spine_05}
   }
   n40 = Anim Dynamics - Bone: dyn_arm_flap_04_L(ComponentPose=@n39, Alpha=1, ExternalForce="0,0,0") {
    angularDampingOverride=0.9999
    angularSpringConstant=50000
    bAngularSpring=true
    bOverrideAngularDamping=true
    simulationSpace=RootRelative
    alphaBoolBlend {blendOption=Linear}
    boundBone {boneName=dyn_arm_flap_04_L}
    componentAppliedLinearAccClamp=<10000 10000 10000>
    componentLinearAccScale=<3 3 3>
    componentLinearVelScale=<3 3 3>
    physicsBodyDefinitions [{
      boundBone {boneName=dyn_arm_flap_04_L}
      boxExtents=<15 15 15>
      constraintSetup {angularLimitsMax=<1 10 10>, angularLimitsMin=<-1 -10 -10>, angularTarget=<0 0 0>}
      localJointOffset=<2 0 0>
     }]
    relativeSpaceBone {boneName=spine_05}
   }
   n41 = Anim Dynamics - Bone: dyn_arm_flap_05_L(ComponentPose=@n40, Alpha=1, ExternalForce="0,0,0") {
    angularDampingOverride=0.99
    angularSpringConstant=50000
    bAngularSpring=true
    bOverrideAngularDamping=true
    simulationSpace=RootRelative
    alphaBoolBlend {blendOption=Linear}
    boundBone {boneName=dyn_arm_flap_05_L}
    componentAppliedLinearAccClamp=<10000 10000 10000>
    componentLinearAccScale=<3 3 3>
    componentLinearVelScale=<3 3 3>
    physicsBodyDefinitions [{
      boundBone {boneName=dyn_arm_flap_05_L}
      boxExtents=<10 10 10>
      constraintSetup {angularLimitsMax=<1 10 10>, angularLimitsMin=<-1 -10 -10>, angularTarget=<0 0 0>}
      localJointOffset=<2 0 0>
     }]
    relativeSpaceBone {boneName=spine_05}
   }
   n42 = Anim Dynamics - Bone: dyn_arm_flap_06_L(ComponentPose=@n41, Alpha=1, ExternalForce="0,0,0") {
    angularDampingOverride=0.9999
    angularSpringConstant=50000
    bAngularSpring=true
    bOverrideAngularDamping=true
    simulationSpace=RootRelative
    alphaBoolBlend {blendOption=Linear}
    boundBone {boneName=dyn_arm_flap_06_L}
    componentAppliedLinearAccClamp=<10000 10000 10000>
    componentLinearAccScale=<3 3 3>
    componentLinearVelScale=<3 3 3>
    physicsBodyDefinitions [{
      boundBone {boneName=dyn_arm_flap_06_L}
      boxExtents=<10 10 10>
      constraintSetup {angularLimitsMax=<1 10 10>, angularLimitsMin=<-1 -10 -10>, angularTarget=<0 0 0>}
      localJointOffset=<2 0 0>
     }]
    relativeSpaceBone {boneName=spine_05}
   }
   n43 = Anim Dynamics - Bone: dyn_arm_flap_07_L(ComponentPose=@n42, Alpha=1, ExternalForce="0,0,0") {
    angularDampingOverride=0.9999
    angularSpringConstant=50000
    bAngularSpring=true
    bOverrideAngularDamping=true
    simulationSpace=RootRelative
    alphaBoolBlend {blendOption=Linear}
    boundBone {boneName=dyn_arm_flap_07_L}
    componentAppliedLinearAccClamp=<10000 10000 10000>
    componentLinearAccScale=<3 3 3>
    componentLinearVelScale=<3 3 3>
    physicsBodyDefinitions [{
      boundBone {boneName=dyn_arm_flap_07_L}
      boxExtents=<10 10 10>
      constraintSetup {angularLimitsMax=<1 10 10>, angularLimitsMin=<-1 -10 -10>, angularTarget=<0 0 0>}
      localJointOffset=<2 0 0>
     }]
    relativeSpaceBone {boneName=spine_05}
   }
   n44 = Anim Dynamics - Bone: dyn_arm_flap_08_L(ComponentPose=@n43, Alpha=1, ExternalForce="0,0,0") {
    angularDampingOverride=0.9999
    angularSpringConstant=50000
    bAngularSpring=true
    bOverrideAngularDamping=true
    simulationSpace=RootRelative
    alphaBoolBlend {blendOption=Linear}
    boundBone {boneName=dyn_arm_flap_08_L}
    componentAppliedLinearAccClamp=<10000 10000 10000>
    componentLinearAccScale=<3 3 3>
    componentLinearVelScale=<3 3 3>
    physicsBodyDefinitions [{
      boundBone {boneName=dyn_arm_flap_08_L}
      boxExtents=<10 10 10>
      constraintSetup {angularLimitsMax=<1 10 10>, angularLimitsMin=<-1 -10 -10>, angularTarget=<0 0 0>}
      localJointOffset=<3 0 0>
     }]
    relativeSpaceBone {boneName=spine_05}
   }
   n45 = Anim Dynamics - Bone: dyn_arm_flap_09_L(ComponentPose=@n44, Alpha=1, ExternalForce="0,0,0") {
    angularDampingOverride=0.99
    angularSpringConstant=50000
    bAngularSpring=true
    bOverrideAngularDamping=true
    simulationSpace=RootRelative
    alphaBoolBlend {blendOption=Linear}
    boundBone {boneName=dyn_arm_flap_09_L}
    componentAppliedLinearAccClamp=<10000 10000 10000>
    componentLinearAccScale=<3 3 3>
    componentLinearVelScale=<3 3 3>
    physicsBodyDefinitions [{
      boundBone {boneName=dyn_arm_flap_09_L}
      boxExtents=<20 20 20>
      constraintSetup {angularLimitsMax=<1 10 10>, angularLimitsMin=<-1 -10 -10>, angularTarget=<0 0 0>}
      localJointOffset=<2 0 0>
     }]
    relativeSpaceBone {boneName=spine_05}
   }
   n46 = Anim Dynamics - Bone: dyn_arm_flap_10_L(ComponentPose=@n45, Alpha=1, ExternalForce="0,0,0") {
    angularDampingOverride=0.9999
    angularSpringConstant=30000
    bAngularSpring=true
    bOverrideAngularDamping=true
    simulationSpace=RootRelative
    alphaBoolBlend {blendOption=Linear}
    boundBone {boneName=dyn_arm_flap_10_L}
    componentAppliedLinearAccClamp=<10000 10000 10000>
    componentLinearAccScale=<2 2 2>
    componentLinearVelScale=<2 2 2>
    physicsBodyDefinitions [{
      boundBone {boneName=dyn_arm_flap_10_L}
      boxExtents=<10 10 10>
      constraintSetup {angularLimitsMax=<1 10 10>, angularLimitsMin=<-1 -10 -10>, angularTarget=<0 0 0>}
      localJointOffset=<2 0 0>
     }]
    relativeSpaceBone {boneName=spine_05}
   }
   n47 = Anim Dynamics - Bone: dyn_arm_flap_11_L(ComponentPose=@n46, Alpha=1, ExternalForce="0,0,0") {
    angularDampingOverride=0.9999
    angularSpringConstant=50000
    bAngularSpring=true
    bOverrideAngularDamping=true
    simulationSpace=RootRelative
    alphaBoolBlend {blendOption=Linear}
    boundBone {boneName=dyn_arm_flap_11_L}
    componentAppliedLinearAccClamp=<10000 10000 10000>
    componentLinearAccScale=<3 3 3>
    componentLinearVelScale=<3 3 3>
    physicsBodyDefinitions [{
      boundBone {boneName=dyn_arm_flap_11_L}
      boxExtents=<10 10 10>
      constraintSetup {angularLimitsMax=<1 10 10>, angularLimitsMin=<-1 -10 -10>, angularTarget=<0 0 0>}
      localJointOffset=<2 0 0>
     }]
    relativeSpaceBone {boneName=spine_05}
   }
   n48 = Anim Dynamics - Bone: dyn_arm_flap_12_L(ComponentPose=@n47, Alpha=1, ExternalForce="0,0,0") {
    angularDampingOverride=1
    angularSpringConstant=50000
    bAngularSpring=true
    bOverrideAngularDamping=true
    simulationSpace=RootRelative
    alphaBoolBlend {blendOption=Linear}
    boundBone {boneName=dyn_arm_flap_12_L}
    componentAppliedLinearAccClamp=<10000 10000 10000>
    componentLinearAccScale=<4 4 4>
    componentLinearVelScale=<4 4 4>
    physicsBodyDefinitions [{
      boundBone {boneName=dyn_arm_flap_12_L}
      boxExtents=<10 10 10>
      constraintSetup {angularLimitsMax=<1 10 10>, angularLimitsMin=<-1 -10 -10>, angularTarget=<0 0 0>}
      localJointOffset=<2 0 0>
     }]
    relativeSpaceBone {boneName=spine_05}
   }
   n49 = Anim Dynamics - Bone: dyn_leg_flap_01_R(ComponentPose=@n48, Alpha=1, ExternalForce="0,0,0") {
    angularDampingOverride=0.9999
    angularSpringConstant=5000
    bAngularSpring=true
    bOverrideAngularDamping=true
    simulationSpace=RootRelative
    alphaBoolBlend {blendOption=Linear}
    boundBone {boneName=dyn_leg_flap_01_R}
    componentAppliedLinearAccClamp=<10000 10000 10000>
    componentLinearAccScale=<3 3 3>
    componentLinearVelScale=<3 3 3>
    physicsBodyDefinitions [{
      boundBone {boneName=dyn_leg_flap_01_R}
      boxExtents=<20 10 10>
      constraintSetup {angularLimitsMax=<0 0 20>, angularLimitsMin=<0 0 -3>, angularTarget=<0 0 0>}
      localJointOffset=<4 0 0>
     }]
    relativeSpaceBone {boneName=pelvis}
   }
   n50 = Anim Dynamics - Bone: dyn_leg_flap_03_R(ComponentPose=@n49, Alpha=1, ExternalForce="0,0,0") {
    angularDampingOverride=0.9999
    angularSpringConstant=5000
    bAngularSpring=true
    bOverrideAngularDamping=true
    simulationSpace=RootRelative
    alphaBoolBlend {blendOption=Linear}
    boundBone {boneName=dyn_leg_flap_03_R}
    componentAppliedLinearAccClamp=<10000 10000 10000>
    componentLinearAccScale=<3 3 3>
    componentLinearVelScale=<3 3 3>
    physicsBodyDefinitions [{
      boundBone {boneName=dyn_leg_flap_03_R}
      boxExtents=<20 10 10>
      constraintSetup {angularLimitsMax=<0 0 20>, angularLimitsMin=<0 0 -3>, angularTarget=<0 0 0>}
      localJointOffset=<4 0 0>
     }]
    relativeSpaceBone {boneName=pelvis}
   }
   n51 = Anim Dynamics - Bone: dyn_leg_flap_05_R(ComponentPose=@n50, Alpha=1, ExternalForce="0,0,0") {
    angularDampingOverride=0.9999
    angularSpringConstant=5000
    bAngularSpring=true
    bOverrideAngularDamping=true
    simulationSpace=RootRelative
    alphaBoolBlend {blendOption=Linear}
    boundBone {boneName=dyn_leg_flap_05_R}
    componentAppliedLinearAccClamp=<10000 10000 10000>
    componentLinearAccScale=<3 3 3>
    componentLinearVelScale=<3 3 3>
    physicsBodyDefinitions [{
      boundBone {boneName=dyn_leg_flap_05_R}
      boxExtents=<20 10 10>
      constraintSetup {angularLimitsMax=<0 0 20>, angularLimitsMin=<0 0 -3>, angularTarget=<0 0 0>}
      localJointOffset=<4 0 0>
     }]
    relativeSpaceBone {boneName=pelvis}
   }
   n52 = Anim Dynamics - Bone: dyn_leg_flap_02_R(ComponentPose=@n51, Alpha=1, ExternalForce="0,0,0") {
    angularDampingOverride=0.9999
    angularSpringConstant=5000
    bAngularSpring=true
    bOverrideAngularDamping=true
    simulationSpace=RootRelative
    alphaBoolBlend {blendOption=Linear}
    boundBone {boneName=dyn_leg_flap_02_R}
    componentAppliedLinearAccClamp=<10000 10000 10000>
    componentLinearAccScale=<3 3 3>
    componentLinearVelScale=<3 3 3>
    physicsBodyDefinitions [{
      boundBone {boneName=dyn_leg_flap_02_R}
      boxExtents=<20 10 10>
      constraintSetup {angularLimitsMax=<0 0 3>, angularLimitsMin=<0 0 -20>, angularTarget=<0 0 0>}
      localJointOffset=<4 0 0>
     }]
    relativeSpaceBone {boneName=pelvis}
   }
   n53 = Anim Dynamics - Bone: dyn_leg_flap_04_R(ComponentPose=@n52, Alpha=1, ExternalForce="0,0,0") {
    angularDampingOverride=0.9999
    angularSpringConstant=5000
    bAngularSpring=true
    bOverrideAngularDamping=true
    simulationSpace=RootRelative
    alphaBoolBlend {blendOption=Linear}
    boundBone {boneName=dyn_leg_flap_04_R}
    componentAppliedLinearAccClamp=<10000 10000 10000>
    componentLinearAccScale=<3 3 3>
    componentLinearVelScale=<3 3 3>
    physicsBodyDefinitions [{
      boundBone {boneName=dyn_leg_flap_04_R}
      boxExtents=<20 10 10>
      constraintSetup {angularLimitsMax=<0 0 3>, angularLimitsMin=<0 0 -20>, angularTarget=<0 0 0>}
      localJointOffset=<4 0 0>
     }]
    relativeSpaceBone {boneName=pelvis}
   }
   n54 = Anim Dynamics - Bone: dyn_leg_flap_06_R(ComponentPose=@n53, Alpha=1, ExternalForce="0,0,0") {
    angularDampingOverride=0.9999
    angularSpringConstant=5000
    bAngularSpring=true
    bOverrideAngularDamping=true
    simulationSpace=RootRelative
    alphaBoolBlend {blendOption=Linear}
    boundBone {boneName=dyn_leg_flap_06_R}
    componentAppliedLinearAccClamp=<10000 10000 10000>
    componentLinearAccScale=<3 3 3>
    componentLinearVelScale=<3 3 3>
    physicsBodyDefinitions [{
      boundBone {boneName=dyn_leg_flap_06_R}
      boxExtents=<20 10 10>
      constraintSetup {angularLimitsMax=<0 0 3>, angularLimitsMin=<0 0 -20>, angularTarget=<0 0 0>}
      localJointOffset=<4 0 0>
     }]
    relativeSpaceBone {boneName=pelvis}
   }
   n55 = Anim Dynamics - Bone: dyn_arm_flap_01_R(ComponentPose=@n54, Alpha=1, ExternalForce="0,0,0") {
    angularDampingOverride=0.99
    angularSpringConstant=50000
    bAngularSpring=true
    bOverrideAngularDamping=true
    simulationSpace=RootRelative
    alphaBoolBlend {blendOption=Linear}
    boundBone {boneName=dyn_arm_flap_01_R}
    componentAppliedLinearAccClamp=<10000 10000 10000>
    componentLinearAccScale=<3 3 3>
    componentLinearVelScale=<3 3 3>
    physicsBodyDefinitions [{
      boundBone {boneName=dyn_arm_flap_01_R}
      boxExtents=<8 8 8>
      constraintSetup {angularLimitsMax=<1 10 10>, angularLimitsMin=<-1 -10 -10>, angularTarget=<0 0 0>}
      localJointOffset=<2 0 0>
     }]
    relativeSpaceBone {boneName=spine_05}
   }
   n56 = Anim Dynamics - Bone: dyn_arm_flap_03_R(ComponentPose=@n55, Alpha=1, ExternalForce="0,0,0") {
    angularDampingOverride=0.9999
    angularSpringConstant=50000
    bAngularSpring=true
    bOverrideAngularDamping=true
    simulationSpace=RootRelative
    alphaBoolBlend {blendOption=Linear}
    boundBone {boneName=dyn_arm_flap_03_R}
    componentAppliedLinearAccClamp=<10000 10000 10000>
    componentLinearAccScale=<3 3 3>
    componentLinearVelScale=<3 3 3>
    physicsBodyDefinitions [{
      boundBone {boneName=dyn_arm_flap_03_R}
      boxExtents=<10 10 10>
      constraintSetup {angularLimitsMax=<1 10 10>, angularLimitsMin=<-1 -10 -10>, angularTarget=<0 0 0>}
      localJointOffset=<2 0 0>
     }]
    relativeSpaceBone {boneName=spine_05}
   }
   n57 = Anim Dynamics - Bone: dyn_arm_flap_03_R(ComponentPose=@n56, Alpha=1, ExternalForce="0,0,0") {
    angularDampingOverride=0.9999
    angularSpringConstant=50000
    bAngularSpring=true
    bOverrideAngularDamping=true
    simulationSpace=RootRelative
    alphaBoolBlend {blendOption=Linear}
    boundBone {boneName=dyn_arm_flap_03_R}
    componentAppliedLinearAccClamp=<10000 10000 10000>
    componentLinearAccScale=<2 2 2>
    componentLinearVelScale=<2 2 2>
    physicsBodyDefinitions [{
      boundBone {boneName=dyn_arm_flap_03_R}
      boxExtents=<15 15 15>
      constraintSetup {angularLimitsMax=<1 10 10>, angularLimitsMin=<-1 -10 -10>, angularTarget=<0 0 0>}
      localJointOffset=<2 0 0>
     }]
    relativeSpaceBone {boneName=spine_05}
   }
   n58 = Anim Dynamics - Bone: dyn_arm_flap_04_R(ComponentPose=@n57, Alpha=1, ExternalForce="0,0,0") {
    angularDampingOverride=0.9999
    angularSpringConstant=50000
    bAngularSpring=true
    bOverrideAngularDamping=true
    simulationSpace=RootRelative
    alphaBoolBlend {blendOption=Linear}
    boundBone {boneName=dyn_arm_flap_04_R}
    componentAppliedLinearAccClamp=<10000 10000 10000>
    componentLinearAccScale=<3 3 3>
    componentLinearVelScale=<3 3 3>
    physicsBodyDefinitions [{
      boundBone {boneName=dyn_arm_flap_04_R}
      boxExtents=<10 10 10>
      constraintSetup {angularLimitsMax=<1 10 10>, angularLimitsMin=<-1 -10 -10>, angularTarget=<0 0 0>}
      localJointOffset=<2 0 0>
     }]
    relativeSpaceBone {boneName=spine_05}
   }
   n59 = Anim Dynamics - Bone: dyn_arm_flap_05_R(ComponentPose=@n58, Alpha=1, ExternalForce="0,0,0") {
    angularDampingOverride=0.9999
    angularSpringConstant=50000
    bAngularSpring=true
    bOverrideAngularDamping=true
    simulationSpace=RootRelative
    alphaBoolBlend {blendOption=Linear}
    boundBone {boneName=dyn_arm_flap_05_R}
    componentAppliedLinearAccClamp=<10000 10000 10000>
    componentLinearAccScale=<3 3 3>
    componentLinearVelScale=<3 3 3>
    physicsBodyDefinitions [{
      boundBone {boneName=dyn_arm_flap_05_R}
      boxExtents=<9 9 9>
      constraintSetup {angularLimitsMax=<1 10 10>, angularLimitsMin=<-1 -10 -10>, angularTarget=<0 0 0>}
      localJointOffset=<3 0 0>
     }]
    relativeSpaceBone {boneName=spine_05}
   }
   n60 = Anim Dynamics - Bone: dyn_arm_flap_06_R(ComponentPose=@n59, Alpha=1, ExternalForce="0,0,0") {
    angularDampingOverride=0.9999
    angularSpringConstant=50000
    bAngularSpring=true
    bOverrideAngularDamping=true
    simulationSpace=RootRelative
    alphaBoolBlend {blendOption=Linear}
    boundBone {boneName=dyn_arm_flap_06_R}
    componentAppliedLinearAccClamp=<10000 10000 10000>
    componentLinearAccScale=<3 3 3>
    componentLinearVelScale=<3 3 3>
    physicsBodyDefinitions [{
      boundBone {boneName=dyn_arm_flap_06_R}
      boxExtents=<10 10 10>
      constraintSetup {angularLimitsMax=<1 10 10>, angularLimitsMin=<-1 -10 -10>, angularTarget=<0 0 0>}
      localJointOffset=<2 0 0>
     }]
    relativeSpaceBone {boneName=spine_05}
   }
   n61 = Anim Dynamics - Bone: dyn_arm_flap_07_R(ComponentPose=@n60, Alpha=1, ExternalForce="0,0,0") {
    angularDampingOverride=0.9999
    angularSpringConstant=50000
    bAngularSpring=true
    bOverrideAngularDamping=true
    simulationSpace=RootRelative
    alphaBoolBlend {blendOption=Linear}
    boundBone {boneName=dyn_arm_flap_07_R}
    componentAppliedLinearAccClamp=<10000 10000 10000>
    componentLinearAccScale=<3 3 3>
    componentLinearVelScale=<3 3 3>
    physicsBodyDefinitions [{
      boundBone {boneName=dyn_arm_flap_07_R}
      boxExtents=<14 14 14>
      constraintSetup {angularLimitsMax=<1 10 10>, angularLimitsMin=<-1 -10 -10>, angularTarget=<0 0 0>}
      localJointOffset=<2 0 0>
     }]
    relativeSpaceBone {boneName=spine_05}
   }
   n62 = Anim Dynamics - Bone: dyn_arm_flap_08_R(ComponentPose=@n61, Alpha=1, ExternalForce="0,0,0") {
    angularDampingOverride=0.99
    angularSpringConstant=50000
    bAngularSpring=true
    bOverrideAngularDamping=true
    simulationSpace=RootRelative
    alphaBoolBlend {blendOption=Linear}
    boundBone {boneName=dyn_arm_flap_08_R}
    componentAppliedLinearAccClamp=<10000 10000 10000>
    componentLinearAccScale=<3 3 3>
    componentLinearVelScale=<3 3 3>
    physicsBodyDefinitions [{
      boundBone {boneName=dyn_arm_flap_08_R}
      boxExtents=<10 10 10>
      constraintSetup {angularLimitsMax=<1 10 10>, angularLimitsMin=<-1 -10 -10>, angularTarget=<0 0 0>}
      localJointOffset=<2 0 0>
     }]
    relativeSpaceBone {boneName=spine_05}
   }
   n63 = Anim Dynamics - Bone: dyn_arm_flap_09_R(ComponentPose=@n62, Alpha=1, ExternalForce="0,0,0") {
    angularDampingOverride=1
    angularSpringConstant=50000
    bAngularSpring=true
    bOverrideAngularDamping=true
    simulationSpace=RootRelative
    alphaBoolBlend {blendOption=Linear}
    boundBone {boneName=dyn_arm_flap_09_R}
    componentAppliedLinearAccClamp=<10000 10000 10000>
    componentLinearAccScale=<4 4 4>
    componentLinearVelScale=<4 4 4>
    physicsBodyDefinitions [{
      boundBone {boneName=dyn_arm_flap_09_R}
      boxExtents=<11 11 11>
      constraintSetup {angularLimitsMax=<1 10 10>, angularLimitsMin=<-1 -10 -10>, angularTarget=<0 0 0>}
      localJointOffset=<2 0 0>
     }]
    relativeSpaceBone {boneName=spine_05}
   }
   n64 = Anim Dynamics - Bone: dyn_arm_flap_10_R(ComponentPose=@n63, Alpha=1, ExternalForce="0,0,0") {
    angularDampingOverride=0.9999
    angularSpringConstant=30000
    bAngularSpring=true
    bOverrideAngularDamping=true
    simulationSpace=RootRelative
    alphaBoolBlend {blendOption=Linear}
    boundBone {boneName=dyn_arm_flap_10_R}
    componentAppliedLinearAccClamp=<10000 10000 10000>
    componentLinearAccScale=<3 3 3>
    componentLinearVelScale=<3 3 3>
    physicsBodyDefinitions [{
      boundBone {boneName=dyn_arm_flap_10_R}
      boxExtents=<10 10 10>
      constraintSetup {angularLimitsMax=<1 10 10>, angularLimitsMin=<-1 -10 -10>, angularTarget=<0 0 0>}
      localJointOffset=<1 0 0>
     }]
    relativeSpaceBone {boneName=spine_05}
   }
   n65 = Anim Dynamics - Bone: dyn_arm_flap_11_R(ComponentPose=@n64, Alpha=1, ExternalForce="0,0,0") {
    angularDampingOverride=0.9999
    angularSpringConstant=50000
    bAngularSpring=true
    bOverrideAngularDamping=true
    simulationSpace=RootRelative
    alphaBoolBlend {blendOption=Linear}
    boundBone {boneName=dyn_arm_flap_11_R}
    componentAppliedLinearAccClamp=<10000 10000 10000>
    componentLinearAccScale=<3 3 3>
    componentLinearVelScale=<3 3 3>
    physicsBodyDefinitions [{
      boundBone {boneName=dyn_arm_flap_11_R}
      boxExtents=<10 10 10>
      constraintSetup {angularLimitsMax=<1 10 10>, angularLimitsMin=<-1 -10 -10>, angularTarget=<0 0 0>}
      localJointOffset=<2 0 0>
     }]
    relativeSpaceBone {boneName=spine_05}
   }
   n66 = Anim Dynamics - Bone: dyn_arm_flap_12_R(ComponentPose=@n65, Alpha=1, ExternalForce="0,0,0") {
    angularDampingOverride=0.9999
    angularSpringConstant=50000
    bAngularSpring=true
    bOverrideAngularDamping=true
    simulationSpace=RootRelative
    alphaBoolBlend {blendOption=Linear}
    boundBone {boneName=dyn_arm_flap_12_R}
    componentAppliedLinearAccClamp=<10000 10000 10000>
    componentLinearAccScale=<3 3 3>
    componentLinearVelScale=<3 3 3>
    physicsBodyDefinitions [{
      boundBone {boneName=dyn_arm_flap_12_R}
      boxExtents=<20 20 20>
      constraintSetup {angularLimitsMax=<1 10 10>, angularLimitsMin=<-1 -10 -10>, angularTarget=<0 0 0>}
      localJointOffset=<1 0 0>
     }]
    relativeSpaceBone {boneName=spine_05}
   }
   n67 = "Constraint - dyn_torsoflap_L_01_mstr (PPPP)"!(
    ComponentPose=@n66
    Alpha=1
    ConstraintWeights_0=0.04
    ConstraintWeights_1:float=0.74
    ConstraintWeights_2:float=0.22
    ConstraintWeights_3:float=0
   ) -> (Pose) {
    alphaBoolBlend {blendOption=Linear}
    boneToModify {boneName=dyn_torsoflap_L_01_mstr}
    constraintSetup [{targetBone {boneName=clavicle_l}}, {targetBone {boneName=spine_04}}, {targetBone {boneName=spine_05}}, {targetBone {boneName=spine_02}}]
    constraintWeights [1,1,1,1]
   }
   n68 = Anim Dynamics - Bone: dyn_torsoflap_L_01(ComponentPose=@n67, Alpha=1, ExternalForce="0,0,0") {
    angularDampingOverride=0.9999
    angularSpringConstant=50000
    bAngularSpring=true
    bOverrideAngularDamping=true
    simulationSpace=RootRelative
    alphaBoolBlend {blendOption=Linear}
    boundBone {boneName=dyn_torsoflap_L_01}
    componentAppliedLinearAccClamp=<10000 10000 10000>
    componentLinearAccScale=<1.5 1.5 1.5>
    componentLinearVelScale=<1.5 1.5 1.5>
    physicsBodyDefinitions [{
      boundBone {boneName=dyn_torsoflap_L_01}
      boxExtents=<10 10 10>
      constraintSetup {angularLimitsMax=<1 10 10>, angularLimitsMin=<-1 -10 -10>, angularTarget=<0 0 0>}
      localJointOffset=<2 0 0>
     }]
    relativeSpaceBone {boneName=spine_05}
   }
   n69 = "Constraint - dyn_torsoflap_L_02_mstr (PPPP)"!(
    ComponentPose=@n68
    Alpha=1
    ConstraintWeights_0=0
    ConstraintWeights_1:float=0.92
    ConstraintWeights_2:float=0.07
    ConstraintWeights_3:float=0.01
   ) -> (Pose) {
    alphaBoolBlend {blendOption=Linear}
    boneToModify {boneName=dyn_torsoflap_L_02_mstr}
    constraintSetup [{targetBone {boneName=spine_01}}, {targetBone {boneName=spine_04}}, {targetBone {boneName=spine_03}}, {targetBone {boneName=spine_02}}]
    constraintWeights [1,1,1,1]
   }
   n70 = Anim Dynamics - Bone: dyn_torsoflap_L_02(ComponentPose=@n69, Alpha=1, ExternalForce="0,0,0") {
    angularDampingOverride=1
    angularSpringConstant=50000
    bAngularSpring=true
    bOverrideAngularDamping=true
    simulationSpace=RootRelative
    alphaBoolBlend {blendOption=Linear}
    boundBone {boneName=dyn_torsoflap_L_02}
    componentAppliedLinearAccClamp=<10000 10000 10000>
    componentLinearAccScale=<2 2 2>
    componentLinearVelScale=<2 2 2>
    physicsBodyDefinitions [{
      boundBone {boneName=dyn_torsoflap_L_02}
      boxExtents=<10 10 10>
      constraintSetup {angularLimitsMax=<1 10 10>, angularLimitsMin=<-1 -10 -10>, angularTarget=<0 0 0>}
      localJointOffset=<2 0 0>
     }]
    relativeSpaceBone {boneName=spine_05}
   }
   n71 = "Constraint - dyn_torsoflap_L_03_mstr (PPPP)"!(
    ComponentPose=@n70
    Alpha=1
    ConstraintWeights_0=0.79
    ConstraintWeights_1:float=0.13
    ConstraintWeights_2:float=0.08
    ConstraintWeights_3:float=0
   ) -> (Pose) {
    alphaBoolBlend {blendOption=Linear}
    boneToModify {boneName=dyn_torsoflap_L_03_mstr}
    constraintSetup [
     {targetBone {boneName=spine_04}}
     {targetBone {boneName=spine_03}}
     {targetBone {boneName=spine_02}}
     {targetBone {boneName=upperarm_twist_01_l}}
    ]
    constraintWeights [1,1,1,1]
   }
   n72 = Anim Dynamics - Bone: dyn_torsoflap_L_03(ComponentPose=@n71, Alpha=1, ExternalForce="0,0,0") {
    angularDampingOverride=1
    angularSpringConstant=50000
    bAngularSpring=true
    bOverrideAngularDamping=true
    simulationSpace=RootRelative
    alphaBoolBlend {blendOption=Linear}
    boundBone {boneName=dyn_torsoflap_L_03}
    componentAppliedLinearAccClamp=<10000 10000 10000>
    componentLinearAccScale=<3 3 3>
    componentLinearVelScale=<3 3 3>
    physicsBodyDefinitions [{
      boundBone {boneName=dyn_torsoflap_L_03}
      boxExtents=<15 15 15>
      constraintSetup {angularLimitsMax=<1 45 10>, angularLimitsMin=<-1 -10 -10>, angularTarget=<0 0 0>}
      localJointOffset=<2 0 0>
     }]
    relativeSpaceBone {boneName=spine_05}
   }
   n73 = "Constraint - dyn_torsoflap_L_04_mstr (PPPP)"!(
    ComponentPose=@n72
    Alpha=1
    ConstraintWeights_0=0.1
    ConstraintWeights_1:float=0.22
    ConstraintWeights_2:float=0.16
    ConstraintWeights_3:float=0.52
   ) -> (Pose) {
    alphaBoolBlend {blendOption=Linear}
    boneToModify {boneName=dyn_torsoflap_L_04_mstr}
    constraintSetup [{targetBone {boneName=spine_01}}, {targetBone {boneName=spine_02}}, {targetBone {boneName=spine_03}}, {targetBone {boneName=spine_04}}]
    constraintWeights [1,1,1,1]
   }
   n74 = Anim Dynamics - Bone: dyn_torsoflap_L_04(ComponentPose=@n73, Alpha=1, ExternalForce="0,0,0") {
    angularDampingOverride=1
    angularSpringConstant=50000
    bAngularSpring=true
    bOverrideAngularDamping=true
    simulationSpace=RootRelative
    alphaBoolBlend {blendOption=Linear}
    boundBone {boneName=dyn_torsoflap_L_04}
    componentAppliedLinearAccClamp=<10000 10000 10000>
    componentLinearAccScale=<3 3 3>
    componentLinearVelScale=<3 3 3>
    physicsBodyDefinitions [{
      boundBone {boneName=dyn_torsoflap_L_04}
      boxExtents=<10 10 10>
      constraintSetup {angularLimitsMax=<1 10 10>, angularLimitsMin=<-1 -10 -10>, angularTarget=<0 0 0>}
      localJointOffset=<2 0 0>
     }]
    relativeSpaceBone {boneName=spine_05}
   }
   n75 = "Constraint - dyn_torsoflap_L_05_mstr (PPPP)"!(
    ComponentPose=@n74
    Alpha=1
    ConstraintWeights_0=0.05
    ConstraintWeights_1:float=0.64
    ConstraintWeights_2:float=0.11
    ConstraintWeights_3:float=0.2
   ) -> (Pose) {
    alphaBoolBlend {blendOption=Linear}
    boneToModify {boneName=dyn_torsoflap_L_05_mstr}
    constraintSetup [{targetBone {boneName=spine_01}}, {targetBone {boneName=spine_04}}, {targetBone {boneName=spine_03}}, {targetBone {boneName=spine_02}}]
    constraintWeights [1,1,1,1]
   }
   n76 = Anim Dynamics - Bone: dyn_torsoflap_L_05(ComponentPose=@n75, Alpha=1, ExternalForce="0,0,0") {
    angularDampingOverride=0.9999
    angularSpringConstant=40000
    bAngularSpring=true
    bOverrideAngularDamping=true
    simulationSpace=RootRelative
    alphaBoolBlend {blendOption=Linear}
    boundBone {boneName=dyn_torsoflap_L_05}
    componentAppliedLinearAccClamp=<10000 10000 10000>
    componentLinearAccScale=<2 2 2>
    componentLinearVelScale=<2 2 2>
    physicsBodyDefinitions [{
      boundBone {boneName=dyn_torsoflap_L_05}
      boxExtents=<10 10 10>
      constraintSetup {angularLimitsMax=<1 10 10>, angularLimitsMin=<-1 -10 -10>, angularTarget=<0 0 0>}
      localJointOffset=<2 0 0>
     }]
    relativeSpaceBone {boneName=spine_05}
   }
   n77 = "Constraint - dyn_torsoflap_L_06_mstr (PPPP)"!(
    ComponentPose=@n76
    Alpha=1
    ConstraintWeights_0=0.13
    ConstraintWeights_1:float=0.44
    ConstraintWeights_2:float=0.17
    ConstraintWeights_3:float=0.26
   ) -> (Pose) {
    alphaBoolBlend {blendOption=Linear}
    boneToModify {boneName=dyn_torsoflap_L_06_mstr}
    constraintSetup [{targetBone {boneName=spine_01}}, {targetBone {boneName=spine_04}}, {targetBone {boneName=spine_03}}, {targetBone {boneName=spine_02}}]
    constraintWeights [1,1,1,1]
   }
   n78 = Anim Dynamics - Bone: dyn_torsoflap_L_06(ComponentPose=@n77, Alpha=1, ExternalForce="0,0,0") {
    angularDampingOverride=1
    angularSpringConstant=50000
    bAngularSpring=true
    bOverrideAngularDamping=true
    simulationSpace=RootRelative
    alphaBoolBlend {blendOption=Linear}
    boundBone {boneName=dyn_torsoflap_L_06}
    componentAppliedLinearAccClamp=<10000 10000 10000>
    componentLinearAccScale=<2 2 2>
    componentLinearVelScale=<2 2 2>
    physicsBodyDefinitions [{
      boundBone {boneName=dyn_torsoflap_L_06}
      boxExtents=<10 10 10>
      constraintSetup {angularLimitsMax=<1 10 10>, angularLimitsMin=<-1 -10 -10>, angularTarget=<0 0 0>}
      localJointOffset=<2 0 0>
     }]
    relativeSpaceBone {boneName=spine_05}
   }
   n79 = "Constraint - dyn_torsoflap_L_07_mstr (PPPPP)"!(
    ComponentPose=@n78
    Alpha=1
    ConstraintWeights_0=0.1
    ConstraintWeights_1:float=0.65
    ConstraintWeights_2:float=0.21
    ConstraintWeights_3:float=0.01
    ConstraintWeights_4:float=0.03
   ) -> (Pose) {
    alphaBoolBlend {blendOption=Linear}
    boneToModify {boneName=dyn_torsoflap_L_07_mstr}
    constraintSetup [
     {targetBone {boneName=clavicle_l}}
     {targetBone {boneName=spine_04}}
     {targetBone {boneName=spine_05}}
     {targetBone {boneName=spine_03}}
     {targetBone {boneName=upperarm_twist_01_l}}
    ]
    constraintWeights [1,1,1,1,1]
   }
   n80 = Anim Dynamics - Bone: dyn_torsoflap_L_07(ComponentPose=@n79, Alpha=1, ExternalForce="0,0,0") {
    angularDampingOverride=1
    angularSpringConstant=50000
    bAngularSpring=true
    bOverrideAngularDamping=true
    simulationSpace=RootRelative
    alphaBoolBlend {blendOption=Linear}
    boundBone {boneName=dyn_torsoflap_L_07}
    componentAppliedLinearAccClamp=<10000 10000 10000>
    componentLinearAccScale=<2 2 2>
    componentLinearVelScale=<2 2 2>
    physicsBodyDefinitions [{
      boundBone {boneName=dyn_torsoflap_L_07}
      boxExtents=<10 10 10>
      constraintSetup {angularLimitsMax=<1 10 10>, angularLimitsMin=<-1 -10 -10>, angularTarget=<0 0 0>}
      localJointOffset=<2 0 0>
     }]
    relativeSpaceBone {boneName=spine_05}
   }
   n81 = "Constraint - dyn_torsoflap_L_08_mstr (PPPP)"!(
    ComponentPose=@n80
    Alpha=1
    ConstraintWeights_0=0.24
    ConstraintWeights_1:float=0.26
    ConstraintWeights_2:float=0.24
    ConstraintWeights_3:float=0.26
   ) -> (Pose) {
    alphaBoolBlend {blendOption=Linear}
    boneToModify {boneName=dyn_torsoflap_L_08_mstr}
    constraintSetup [
     {targetBone {boneName=clavicle_l}}
     {targetBone {boneName=spine_04}}
     {targetBone {boneName=spine_05}}
     {targetBone {boneName=upperarm_twist_01_l}}
    ]
    constraintWeights [1,1,1,1]
   }
   n82 = Anim Dynamics - Bone: dyn_torsoflap_L_08(ComponentPose=@n81, Alpha=1, ExternalForce="0,0,0") {
    angularDampingOverride=1
    angularSpringConstant=30000
    bAngularSpring=true
    bOverrideAngularDamping=true
    simulationSpace=RootRelative
    alphaBoolBlend {blendOption=Linear}
    boundBone {boneName=dyn_torsoflap_L_08}
    componentAppliedLinearAccClamp=<10000 10000 10000>
    componentLinearAccScale=<2 2 2>
    componentLinearVelScale=<2 2 2>
    physicsBodyDefinitions [{
      boundBone {boneName=dyn_torsoflap_L_08}
      boxExtents=<15 15 15>
      constraintSetup {angularLimitsMax=<1 30 10>, angularLimitsMin=<-1 -20 -10>, angularTarget=<0 0 0>}
      localJointOffset=<2 0 0>
     }]
    relativeSpaceBone {boneName=spine_05}
   }
   n83 = "Constraint - dyn_torsoflap_L_09_mstr (PPPPP)"!(
    ComponentPose=@n82
    Alpha=1
    ConstraintWeights_0=0.04
    ConstraintWeights_1:float=0.54
    ConstraintWeights_2:float=0.2
    ConstraintWeights_3:float=0.15
    ConstraintWeights_4:float=0.07
   ) -> (Pose) {
    alphaBoolBlend {blendOption=Linear}
    boneToModify {boneName=dyn_torsoflap_L_09_mstr}
    constraintSetup [
     {targetBone {boneName=clavicle_l}}
     {targetBone {boneName=spine_04}}
     {targetBone {boneName=spine_05}}
     {targetBone {boneName=spine_03}}
     {targetBone {boneName=spine_02}}
    ]
    constraintWeights [1,1,1,1,1]
   }
   n84 = Anim Dynamics - Bone: dyn_torsoflap_L_09(ComponentPose=@n83, Alpha=1, ExternalForce="0,0,0") {
    angularDampingOverride=0.9999
    angularSpringConstant=50000
    bAngularSpring=true
    bOverrideAngularDamping=true
    simulationSpace=RootRelative
    alphaBoolBlend {blendOption=Linear}
    boundBone {boneName=dyn_torsoflap_L_09}
    componentAppliedLinearAccClamp=<10000 10000 10000>
    componentLinearAccScale=<2 2 2>
    componentLinearVelScale=<2 2 2>
    physicsBodyDefinitions [{
      boundBone {boneName=dyn_torsoflap_L_09}
      boxExtents=<10 10 10>
      constraintSetup {angularLimitsMax=<1 10 10>, angularLimitsMin=<-1 -10 -10>, angularTarget=<0 0 0>}
      localJointOffset=<2 0 0>
     }]
    relativeSpaceBone {boneName=spine_05}
   }
   n85 = "Constraint - dyn_torsoflap_L_10_mstr (PPPPP)"!(
    ComponentPose=@n84
    Alpha=1
    ConstraintWeights_0=0.36
    ConstraintWeights_1:float=0.3
    ConstraintWeights_2:float=0.27
    ConstraintWeights_3:float=0
    ConstraintWeights_4:float=0.07
   ) -> (Pose) {
    alphaBoolBlend {blendOption=Linear}
    boneToModify {boneName=dyn_torsoflap_L_10_mstr}
    constraintSetup [
     {targetBone {boneName=clavicle_l}}
     {targetBone {boneName=spine_04}}
     {targetBone {boneName=spine_05}}
     {targetBone {boneName=spine_03}}
     {targetBone {boneName=upperarm_twist_01_l}}
    ]
    constraintWeights [1,1,1,1,1]
   }
   n86 = Anim Dynamics - Bone: dyn_torsoflap_L_10(ComponentPose=@n85, Alpha=1, ExternalForce="0,0,0") {
    angularDampingOverride=1
    angularSpringConstant=50000
    bAngularSpring=true
    bOverrideAngularDamping=true
    simulationSpace=RootRelative
    alphaBoolBlend {blendOption=Linear}
    boundBone {boneName=dyn_torsoflap_L_10}
    componentAppliedLinearAccClamp=<10000 10000 10000>
    componentLinearAccScale=<2 2 2>
    componentLinearVelScale=<2 2 2>
    physicsBodyDefinitions [{
      boundBone {boneName=dyn_torsoflap_L_10}
      boxExtents=<10 10 10>
      constraintSetup {angularLimitsMax=<1 10 10>, angularLimitsMin=<-1 -10 -10>, angularTarget=<0 0 0>}
      localJointOffset=<2 0 0>
     }]
    relativeSpaceBone {boneName=spine_05}
   }
   n87 = "Constraint - dyn_torsoflap_L_12_mstr (PPPP)"!(
    ComponentPose=@n86
    Alpha=1
    ConstraintWeights_0=0.34
    ConstraintWeights_1:float=0.27
    ConstraintWeights_2:float=0.29
    ConstraintWeights_3:float=0.1
   ) -> (Pose) {
    alphaBoolBlend {blendOption=Linear}
    boneToModify {boneName=dyn_torsoflap_L_12_mstr}
    constraintSetup [
     {targetBone {boneName=spine_04}}
     {targetBone {boneName=spine_05}}
     {targetBone {boneName=clavicle_l}}
     {targetBone {boneName=upperarm_twist_01_l}}
    ]
    constraintWeights [1,1,1,1]
   }
   n88 = Anim Dynamics - Bone: dyn_torsoflap_L_12(ComponentPose=@n87, Alpha=1, ExternalForce="0,0,0") {
    angularDampingOverride=1
    angularSpringConstant=50000
    bAngularSpring=true
    bOverrideAngularDamping=true
    simulationSpace=RootRelative
    alphaBoolBlend {blendOption=Linear}
    boundBone {boneName=dyn_torsoflap_L_12}
    componentAppliedLinearAccClamp=<10000 10000 10000>
    componentLinearAccScale=<2 2 2>
    componentLinearVelScale=<2 2 2>
    physicsBodyDefinitions [{
      boundBone {boneName=dyn_torsoflap_L_12}
      boxExtents=<8 8 8>
      constraintSetup {angularLimitsMax=<1 10 10>, angularLimitsMin=<-1 -10 -10>, angularTarget=<0 0 0>}
      localJointOffset=<2 0 0>
     }]
    relativeSpaceBone {boneName=spine_05}
   }
   n89 = "Constraint - dyn_torsoflap_L_11_mstr (PPPP)"!(
    ComponentPose=@n88
    Alpha=1
    ConstraintWeights_0=0.34
    ConstraintWeights_1:float=0.27
    ConstraintWeights_2:float=0.29
    ConstraintWeights_3:float=0.1
   ) -> (Pose) {
    alphaBoolBlend {blendOption=Linear}
    boneToModify {boneName=dyn_torsoflap_L_11_mstr}
    constraintSetup [
     {targetBone {boneName=spine_04}}
     {targetBone {boneName=spine_05}}
     {targetBone {boneName=clavicle_l}}
     {targetBone {boneName=upperarm_twist_01_l}}
    ]
    constraintWeights [1,1,1,1]
   }
   n90 = Anim Dynamics - Bone: dyn_torsoflap_L_11(ComponentPose=@n89, Alpha=1, ExternalForce="0,0,0") {
    angularDampingOverride=1
    angularSpringConstant=50000
    bAngularSpring=true
    bOverrideAngularDamping=true
    simulationSpace=RootRelative
    alphaBoolBlend {blendOption=Linear}
    boundBone {boneName=dyn_torsoflap_L_11}
    componentAppliedLinearAccClamp=<10000 10000 10000>
    componentLinearAccScale=<3 3 3>
    componentLinearVelScale=<3 3 3>
    physicsBodyDefinitions [{
      boundBone {boneName=dyn_torsoflap_L_11}
      boxExtents=<10 10 10>
      constraintSetup {angularLimitsMax=<1 10 10>, angularLimitsMin=<-1 -10 -10>, angularTarget=<0 0 0>}
      localJointOffset=<2 0 0>
     }]
    relativeSpaceBone {boneName=spine_05}
   }
   n91 = "Constraint - dyn_torsoflap_R_01_mstr (PPPP)"!(
    ComponentPose=@n90
    Alpha=1
    ConstraintWeights_0=0.04
    ConstraintWeights_1:float=0.74
    ConstraintWeights_2:float=0.22
    ConstraintWeights_3:float=0
   ) -> (Pose) {
    alphaBoolBlend {blendOption=Linear}
    boneToModify {boneName=dyn_torsoflap_R_01_mstr}
    constraintSetup [{targetBone {boneName=clavicle_r}}, {targetBone {boneName=spine_04}}, {targetBone {boneName=spine_05}}, {targetBone {boneName=spine_02}}]
    constraintWeights [1,1,1,1]
   }
   n92 = Anim Dynamics - Bone: dyn_torsoflap_R_01(ComponentPose=@n91, Alpha=1, ExternalForce="0,0,0") {
    angularDampingOverride=1
    angularSpringConstant=50000
    bAngularSpring=true
    bOverrideAngularDamping=true
    simulationSpace=RootRelative
    alphaBoolBlend {blendOption=Linear}
    boundBone {boneName=dyn_torsoflap_R_01}
    componentAppliedLinearAccClamp=<10000 10000 10000>
    componentLinearAccScale=<2 2 2>
    componentLinearVelScale=<2 2 2>
    physicsBodyDefinitions [{
      boundBone {boneName=dyn_torsoflap_R_01}
      boxExtents=<10 10 10>
      constraintSetup {angularLimitsMax=<1 10 10>, angularLimitsMin=<-1 -10 -10>, angularTarget=<0 0 0>}
      localJointOffset=<2 0 0>
     }]
    relativeSpaceBone {boneName=spine_05}
   }
   n93 = "Constraint - dyn_torsoflap_R_02_mstr (PPPP)"!(
    ComponentPose=@n92
    Alpha=1
    ConstraintWeights_0=0
    ConstraintWeights_1:float=0.92
    ConstraintWeights_2:float=0.07
    ConstraintWeights_3:float=0.01
   ) -> (Pose) {
    alphaBoolBlend {blendOption=Linear}
    boneToModify {boneName=dyn_torsoflap_R_02_mstr}
    constraintSetup [{targetBone {boneName=spine_01}}, {targetBone {boneName=spine_04}}, {targetBone {boneName=spine_03}}, {targetBone {boneName=spine_02}}]
    constraintWeights [1,1,1,1]
   }
   n94 = Anim Dynamics - Bone: dyn_torsoflap_R_02(ComponentPose=@n93, Alpha=1, ExternalForce="0,0,0") {
    angularDampingOverride=1
    angularSpringConstant=50000
    bAngularSpring=true
    bOverrideAngularDamping=true
    simulationSpace=RootRelative
    alphaBoolBlend {blendOption=Linear}
    boundBone {boneName=dyn_torsoflap_R_02}
    componentAppliedLinearAccClamp=<10000 10000 10000>
    componentLinearAccScale=<2 2 2>
    componentLinearVelScale=<2 2 2>
    physicsBodyDefinitions [{
      boundBone {boneName=dyn_torsoflap_R_02}
      boxExtents=<10 10 10>
      constraintSetup {angularLimitsMax=<1 10 10>, angularLimitsMin=<-1 -10 -10>, angularTarget=<0 0 0>}
      localJointOffset=<2 0 0>
     }]
    relativeSpaceBone {boneName=spine_05}
   }
   n95 = "Constraint - dyn_torsoflap_R_03_mstr (PPPP)"!(
    ComponentPose=@n94
    Alpha=1
    ConstraintWeights_0=0.79
    ConstraintWeights_1:float=0.13
    ConstraintWeights_2:float=0.08
    ConstraintWeights_3:float=0
   ) -> (Pose) {
    alphaBoolBlend {blendOption=Linear}
    boneToModify {boneName=dyn_torsoflap_R_03_mstr}
    constraintSetup [
     {targetBone {boneName=spine_04}}
     {targetBone {boneName=spine_03}}
     {targetBone {boneName=spine_02}}
     {targetBone {boneName=upperarm_twist_01_l}}
    ]
    constraintWeights [1,1,1,1]
   }
   n96 = Anim Dynamics - Bone: dyn_torsoflap_R_03(ComponentPose=@n95, Alpha=1, ExternalForce="0,0,0") {
    angularDampingOverride=1
    angularSpringConstant=50000
    bAngularSpring=true
    bOverrideAngularDamping=true
    simulationSpace=RootRelative
    alphaBoolBlend {blendOption=Linear}
    boundBone {boneName=dyn_torsoflap_R_03}
    componentAppliedLinearAccClamp=<10000 10000 10000>
    componentLinearAccScale=<3 3 3>
    componentLinearVelScale=<3 3 3>
    physicsBodyDefinitions [{
      boundBone {boneName=dyn_torsoflap_R_03}
      boxExtents=<10 10 10>
      constraintSetup {angularLimitsMax=<1 45 10>, angularLimitsMin=<-1 -10 -10>, angularTarget=<0 0 0>}
      localJointOffset=<2 0 0>
     }]
    relativeSpaceBone {boneName=spine_05}
   }
   n97 = "Constraint - dyn_torsoflap_R_04_mstr (PPPP)"!(
    ComponentPose=@n96
    Alpha=1
    ConstraintWeights_0=0.1
    ConstraintWeights_1:float=0.22
    ConstraintWeights_2:float=0.16
    ConstraintWeights_3:float=0.52
   ) -> (Pose) {
    alphaBoolBlend {blendOption=Linear}
    boneToModify {boneName=dyn_torsoflap_R_04_mstr}
    constraintSetup [{targetBone {boneName=spine_01}}, {targetBone {boneName=spine_02}}, {targetBone {boneName=spine_03}}, {targetBone {boneName=spine_04}}]
    constraintWeights [1,1,1,1]
   }
   n98 = Anim Dynamics - Bone: dyn_torsoflap_R_04(ComponentPose=@n97, Alpha=1, ExternalForce="0,0,0") {
    angularDampingOverride=1
    angularSpringConstant=50000
    bAngularSpring=true
    bOverrideAngularDamping=true
    simulationSpace=RootRelative
    alphaBoolBlend {blendOption=Linear}
    boundBone {boneName=dyn_torsoflap_R_04}
    componentAppliedLinearAccClamp=<10000 10000 10000>
    componentLinearAccScale=<2 2 2>
    componentLinearVelScale=<2 2 2>
    physicsBodyDefinitions [{
      boundBone {boneName=dyn_torsoflap_R_04}
      boxExtents=<10 10 10>
      constraintSetup {angularLimitsMax=<1 10 10>, angularLimitsMin=<-1 -10 -10>, angularTarget=<0 0 0>}
      localJointOffset=<2 0 0>
     }]
    relativeSpaceBone {boneName=spine_05}
   }
   n99 = "Constraint - dyn_torsoflap_R_05_mstr (PPPP)"!(
    ComponentPose=@n98
    Alpha=1
    ConstraintWeights_0=0.05
    ConstraintWeights_1:float=0.64
    ConstraintWeights_2:float=0.11
    ConstraintWeights_3:float=0.2
   ) -> (Pose) {
    alphaBoolBlend {blendOption=Linear}
    boneToModify {boneName=dyn_torsoflap_R_05_mstr}
    constraintSetup [{targetBone {boneName=spine_01}}, {targetBone {boneName=spine_04}}, {targetBone {boneName=spine_03}}, {targetBone {boneName=spine_02}}]
    constraintWeights [1,1,1,1]
   }
   n100 = Anim Dynamics - Bone: dyn_torsoflap_R_05(ComponentPose=@n99, Alpha=1, ExternalForce="0,0,0") {
    angularDampingOverride=1
    angularSpringConstant=50000
    bAngularSpring=true
    bOverrideAngularDamping=true
    simulationSpace=RootRelative
    alphaBoolBlend {blendOption=Linear}
    boundBone {boneName=dyn_torsoflap_R_05}
    componentAppliedLinearAccClamp=<10000 10000 10000>
    componentLinearAccScale=<2 2 2>
    componentLinearVelScale=<2 2 2>
    physicsBodyDefinitions [{
      boundBone {boneName=dyn_torsoflap_R_05}
      boxExtents=<10 10 10>
      constraintSetup {angularLimitsMax=<1 10 10>, angularLimitsMin=<-1 -10 -10>, angularTarget=<0 0 0>}
      localJointOffset=<2 0 0>
     }]
    relativeSpaceBone {boneName=spine_05}
   }
   n101 = "Constraint - dyn_torsoflap_R_06_mstr (PPPP)"!(
    ComponentPose=@n100
    Alpha=1
    ConstraintWeights_0=0.13
    ConstraintWeights_1:float=0.44
    ConstraintWeights_2:float=0.17
    ConstraintWeights_3:float=0.26
   ) -> (Pose) {
    alphaBoolBlend {blendOption=Linear}
    boneToModify {boneName=dyn_torsoflap_R_06_mstr}
    constraintSetup [{targetBone {boneName=spine_01}}, {targetBone {boneName=spine_04}}, {targetBone {boneName=spine_03}}, {targetBone {boneName=spine_02}}]
    constraintWeights [1,1,1,1]
   }
   n102 = Anim Dynamics - Bone: dyn_torsoflap_R_06(ComponentPose=@n101, Alpha=1, ExternalForce="0,0,0") {
    angularDampingOverride=1
    angularSpringConstant=50000
    bAngularSpring=true
    bOverrideAngularDamping=true
    simulationSpace=RootRelative
    alphaBoolBlend {blendOption=Linear}
    boundBone {boneName=dyn_torsoflap_R_06}
    componentAppliedLinearAccClamp=<10000 10000 10000>
    componentLinearAccScale=<2 2 2>
    componentLinearVelScale=<2 2 2>
    physicsBodyDefinitions [{
      boundBone {boneName=dyn_torsoflap_R_06}
      boxExtents=<10 10 10>
      constraintSetup {angularLimitsMax=<1 10 10>, angularLimitsMin=<-1 -10 -10>, angularTarget=<0 0 0>}
      localJointOffset=<2 0 0>
     }]
    relativeSpaceBone {boneName=spine_05}
   }
   n103 = "Constraint - dyn_torsoflap_R_07_mstr (PPPPP)"!(
    ComponentPose=@n102
    Alpha=1
    ConstraintWeights_0=0.1
    ConstraintWeights_1:float=0.65
    ConstraintWeights_2:float=0.21
    ConstraintWeights_3:float=0.01
    ConstraintWeights_4:float=0.03
   ) -> (Pose) {
    alphaBoolBlend {blendOption=Linear}
    boneToModify {boneName=dyn_torsoflap_R_07_mstr}
    constraintSetup [
     {targetBone {boneName=clavicle_r}}
     {targetBone {boneName=spine_04}}
     {targetBone {boneName=spine_05}}
     {targetBone {boneName=spine_03}}
     {targetBone {boneName=upperarm_twist_01_r}}
    ]
    constraintWeights [1,1,1,1,1]
   }
   n104 = Anim Dynamics - Bone: dyn_torsoflap_R_07(ComponentPose=@n103, Alpha=1, ExternalForce="0,0,0") {
    angularDampingOverride=1
    angularSpringConstant=50000
    bAngularSpring=true
    bOverrideAngularDamping=true
    simulationSpace=RootRelative
    alphaBoolBlend {blendOption=Linear}
    boundBone {boneName=dyn_torsoflap_R_07}
    componentAppliedLinearAccClamp=<10000 10000 10000>
    componentLinearAccScale=<2 2 2>
    componentLinearVelScale=<2 2 2>
    physicsBodyDefinitions [{
      boundBone {boneName=dyn_torsoflap_R_07}
      boxExtents=<10 10 10>
      constraintSetup {angularLimitsMax=<1 10 10>, angularLimitsMin=<-1 -10 -10>, angularTarget=<0 0 0>}
      localJointOffset=<2 0 0>
     }]
    relativeSpaceBone {boneName=spine_05}
   }
   n105 = "Constraint - dyn_torsoflap_R_08_mstr (PPPP)"!(
    ComponentPose=@n104
    Alpha=1
    ConstraintWeights_0=0.24
    ConstraintWeights_1:float=0.26
    ConstraintWeights_2:float=0.24
    ConstraintWeights_3:float=0.26
   ) -> (Pose) {
    alphaBoolBlend {blendOption=Linear}
    boneToModify {boneName=dyn_torsoflap_R_08_mstr}
    constraintSetup [
     {targetBone {boneName=clavicle_r}}
     {targetBone {boneName=spine_04}}
     {targetBone {boneName=spine_05}}
     {targetBone {boneName=upperarm_twist_01_r}}
    ]
    constraintWeights [1,1,1,1]
   }
   n106 = Anim Dynamics - Bone: dyn_torsoflap_R_08(ComponentPose=@n105, Alpha=1, ExternalForce="0,0,0") {
    angularDampingOverride=1
    angularSpringConstant=50000
    bAngularSpring=true
    bOverrideAngularDamping=true
    simulationSpace=RootRelative
    alphaBoolBlend {blendOption=Linear}
    boundBone {boneName=dyn_torsoflap_R_08}
    componentAppliedLinearAccClamp=<10000 10000 10000>
    componentLinearAccScale=<2 2 2>
    componentLinearVelScale=<2 2 2>
    physicsBodyDefinitions [{
      boundBone {boneName=dyn_torsoflap_R_08}
      boxExtents=<10 10 10>
      constraintSetup {angularLimitsMax=<1 30 10>, angularLimitsMin=<-1 -20 -10>, angularTarget=<0 0 0>}
      localJointOffset=<2 0 0>
     }]
    relativeSpaceBone {boneName=spine_05}
   }
   n107 = "Constraint - dyn_torsoflap_R_09_mstr (PPPPP)"!(
    ComponentPose=@n106
    Alpha=1
    ConstraintWeights_0=0.04
    ConstraintWeights_1:float=0.54
    ConstraintWeights_2:float=0.2
    ConstraintWeights_3:float=0.15
    ConstraintWeights_4:float=0.07
   ) -> (Pose) {
    alphaBoolBlend {blendOption=Linear}
    boneToModify {boneName=dyn_torsoflap_R_09_mstr}
    constraintSetup [
     {targetBone {boneName=clavicle_r}}
     {targetBone {boneName=spine_04}}
     {targetBone {boneName=spine_05}}
     {targetBone {boneName=spine_03}}
     {targetBone {boneName=spine_02}}
    ]
    constraintWeights [1,1,1,1,1]
   }
   n108 = Anim Dynamics - Bone: dyn_torsoflap_R_09(ComponentPose=@n107, Alpha=1, ExternalForce="0,0,0") {
    angularDampingOverride=1
    angularSpringConstant=50000
    bAngularSpring=true
    bOverrideAngularDamping=true
    simulationSpace=RootRelative
    alphaBoolBlend {blendOption=Linear}
    boundBone {boneName=dyn_torsoflap_R_09}
    componentAppliedLinearAccClamp=<10000 10000 10000>
    componentLinearAccScale=<2 2 2>
    componentLinearVelScale=<2 2 2>
    physicsBodyDefinitions [{
      boundBone {boneName=dyn_torsoflap_R_09}
      boxExtents=<10 10 10>
      constraintSetup {angularLimitsMax=<1 10 10>, angularLimitsMin=<-1 -10 -10>, angularTarget=<0 0 0>}
      localJointOffset=<2 0 0>
     }]
    relativeSpaceBone {boneName=spine_05}
   }
   n109 = "Constraint - dyn_torsoflap_R_10_mstr (PPPPP)"!(
    ComponentPose=@n108
    Alpha=1
    ConstraintWeights_0=0.36
    ConstraintWeights_1:float=0.3
    ConstraintWeights_2:float=0.27
    ConstraintWeights_3:float=0
    ConstraintWeights_4:float=0.07
   ) -> (Pose) {
    alphaBoolBlend {blendOption=Linear}
    boneToModify {boneName=dyn_torsoflap_R_10_mstr}
    constraintSetup [
     {targetBone {boneName=clavicle_r}}
     {targetBone {boneName=spine_04}}
     {targetBone {boneName=spine_05}}
     {targetBone {boneName=spine_03}}
     {targetBone {boneName=upperarm_twist_01_r}}
    ]
    constraintWeights [1,1,1,1,1]
   }
   n110 = Anim Dynamics - Bone: dyn_torsoflap_R_10(ComponentPose=@n109, Alpha=1, ExternalForce="0,0,0") {
    angularDampingOverride=1
    angularSpringConstant=50000
    bAngularSpring=true
    bOverrideAngularDamping=true
    simulationSpace=RootRelative
    alphaBoolBlend {blendOption=Linear}
    boundBone {boneName=dyn_torsoflap_R_10}
    componentAppliedLinearAccClamp=<10000 10000 10000>
    componentLinearAccScale=<2 2 2>
    componentLinearVelScale=<2 2 2>
    physicsBodyDefinitions [{
      boundBone {boneName=dyn_torsoflap_R_10}
      boxExtents=<10 10 10>
      constraintSetup {angularLimitsMax=<1 10 10>, angularLimitsMin=<-1 -10 -10>, angularTarget=<0 0 0>}
      localJointOffset=<2 0 0>
     }]
    relativeSpaceBone {boneName=spine_05}
   }
   n111 = "Constraint - dyn_torsoflap_R_12_mstr (PPPP)"!(
    ComponentPose=@n110
    Alpha=1
    ConstraintWeights_0=0.34
    ConstraintWeights_1:float=0.27
    ConstraintWeights_2:float=0.29
    ConstraintWeights_3:float=0.1
   ) -> (Pose) {
    alphaBoolBlend {blendOption=Linear}
    boneToModify {boneName=dyn_torsoflap_R_12_mstr}
    constraintSetup [
     {targetBone {boneName=spine_04}}
     {targetBone {boneName=spine_05}}
     {targetBone {boneName=clavicle_r}}
     {targetBone {boneName=upperarm_twist_01_r}}
    ]
    constraintWeights [1,1,1,1]
   }
   n112 = Anim Dynamics - Bone: dyn_torsoflap_R_12(ComponentPose=@n111, Alpha=1, ExternalForce="0,0,0") {
    angularDampingOverride=1
    angularSpringConstant=50000
    bAngularSpring=true
    bOverrideAngularDamping=true
    simulationSpace=RootRelative
    alphaBoolBlend {blendOption=Linear}
    boundBone {boneName=dyn_torsoflap_R_12}
    componentAppliedLinearAccClamp=<10000 10000 10000>
    componentLinearAccScale=<2 2 2>
    componentLinearVelScale=<2 2 2>
    physicsBodyDefinitions [{
      boundBone {boneName=dyn_torsoflap_R_12}
      boxExtents=<10 10 10>
      constraintSetup {angularLimitsMax=<1 10 10>, angularLimitsMin=<-1 -10 -10>, angularTarget=<0 0 0>}
      localJointOffset=<2 0 0>
     }]
    relativeSpaceBone {boneName=spine_05}
   }
   n113 = "Constraint - dyn_torsoflap_R_11_mstr (PPPP)"!(
    ComponentPose=@n112
    Alpha=1
    ConstraintWeights_0=0.34
    ConstraintWeights_1:float=0.27
    ConstraintWeights_2:float=0.29
    ConstraintWeights_3:float=0.1
   ) -> (Pose) {
    alphaBoolBlend {blendOption=Linear}
    boneToModify {boneName=dyn_torsoflap_R_11_mstr}
    constraintSetup [
     {targetBone {boneName=spine_04}}
     {targetBone {boneName=spine_05}}
     {targetBone {boneName=clavicle_r}}
     {targetBone {boneName=upperarm_twist_01_r}}
    ]
    constraintWeights [1,1,1,1]
   }
   n114 = Anim Dynamics - Bone: dyn_torsoflap_R_11(ComponentPose=@n113, Alpha=1, ExternalForce="0,0,0") {
    angularDampingOverride=1
    angularSpringConstant=50000
    bAngularSpring=true
    bOverrideAngularDamping=true
    simulationSpace=RootRelative
    alphaBoolBlend {blendOption=Linear}
    boundBone {boneName=dyn_torsoflap_R_11}
    componentAppliedLinearAccClamp=<10000 10000 10000>
    componentLinearAccScale=<2 2 2>
    componentLinearVelScale=<2 2 2>
    physicsBodyDefinitions [{
      boundBone {boneName=dyn_torsoflap_R_11}
      boxExtents=<10 10 10>
      constraintSetup {angularLimitsMax=<1 10 10>, angularLimitsMin=<-1 -10 -10>, angularTarget=<0 0 0>}
      localJointOffset=<2 0 0>
     }]
    relativeSpaceBone {boneName=spine_05}
   }
   n115 = Control Rig!(Source=Component To Local(ComponentPose=@n114), Alpha, IsCrouchingCtrl:bool=false, RbanOffCtrl:bool=false) -> (Pose) {
    alphaBoolBlend {blendOption=Linear}
    controlRigAssetReference {blueprintRigClass=&/F_MED_RainbowStraps_CtrlRig.F_MED_RainbowStraps_CtrlRig_C}
   }
   Output Pose(Result=@n115) {layerGroup=None}
   Get bIsSurfaceSwimming()
  ]}
 EventGraph {nodes [
   n1 = Event BlueprintUpdateAnimation()
   > n2 = Sequence!(execute) -> (then_0, then_1, then_2:Exec)
   > n3 = Gravity Override State Function DEPRECATED(
    CharacterPartAnimInstance=Self-Reference()
    GameTimeInSeconds=GetGameTimeInSeconds()
    GravityOverrideParams=Get GravityOverrideParams() -> (GravityOverrideParams:Gravity Override Params Struct Structure)
   )
   > Set GravityOverride(GravityOverride=@n3)
   Is Valid(exec=@n2.then_1, InputObject=Get PlayerPawn() -> (PlayerPawn:Fort Player Pawn Athena Object Reference))
   > Set WPO Globals Function(
    Material=Get Mat Instance Dynamic()
    Overlay Material=Get Overlay Mat Instance Dynamic()
    Skeletal Mesh=Get SkelMesh()
    "Local Noise Transform  Location"=Get Local Noise Transform Location() -> (Local Noise Transform Location:Vector)
    Local Noise Offset=Get Local Noise Offset()
   )
   > Set WPO Regional Function!(
    execute
    Material:Material Instance Dynamic Object Reference=Get Mat Instance Dynamic()
    Overlay Material:Material Instance Dynamic Object Reference=Get Overlay Mat Instance Dynamic()
    Skeletal Mesh:Skeletal Mesh Component Object Reference=Get SkelMesh()
    Velocity Transform:Name=foot_l
    Velocity Interp Speed:double=Get Velocity Interp Speed()
    Collision Transform:Name=calf_l
    Collision Transform Axis:EAxis Enum
    Invert Collision Transform Axis:bool=false
    Delta Time X:double=@n1.DeltaTimeX
    Previous Velocity Transform Location:"Vector (by ref)"=Get R1 Velocity Transform Location()
    Previous Velocity Transform Velocity:"Vector (by ref)"=Get R1 Velocity Transform Velocity()
    Region:int=1
   ) -> (then)
   > Set WPO Regional Function(
    Material=Get Mat Instance Dynamic()
    Overlay Material=Get Overlay Mat Instance Dynamic()
    Skeletal Mesh=Get SkelMesh()
    Velocity Transform=foot_r
    Velocity Interp Speed=Get Velocity Interp Speed()
    Collision Transform=calf_r
    Invert Collision Transform Axis=true
    Delta Time X=@n1.DeltaTimeX
    Previous Velocity Transform Location=Get R2 Velocity Transform Location()
    Previous Velocity Transform Velocity=Get R2 Velocity Transform Velocity()
    Region=2
   )
   > n4 = Branch(Condition=Get bIsSkydiveFloating())
   > n5 = Set WPO Pho Wind Intensity(WPO Pho Wind Intensity=1)
   SetScalarParameterValue(execute=[@n6.then, @n5.then], self=Get Mat Instance Dynamic(), ParameterName=ClothSim_WindStrength, Value=Get WPO Pho Wind Intensity())
   n6 = Set WPO Pho Wind Intensity(execute=@n4.else, WPO Pho Wind Intensity=0)
   n7 = Get bIsInFrontEnd()
   Set NotFrontEnd(execute=@n2.then_2, NotFrontEnd:double=SelectFloat(A=1, B=0, bPickA="Not Equal (Boolean)"(A=@n7, B=true) -> (ReturnValue=~))) -> (Output_Get:double)
   Event BlueprintInitializeAnimation()
   > Set SkelMesh(SkelMesh=GetOwnerSkeletalMeshComponent())
   n8 = TryGetPawnOwner()
   > Is Valid(InputObject=@n8)
   > n9 = Cast To FortPlayerPawnAthena(Object=@n8)
   > Set PlayerPawn(PlayerPawn:Fort Player Pawn Athena Object Reference=@n9) -> (Output_Get:Fort Player Pawn Athena Object Reference)
  ]}
]
variables [
 GravityOverride {type=Vector}
 GravityOverrideParams {type=GravityOverrideParamsStruct}
 Local Noise Offset {type=Vector}
 Local Noise Transform Location {type=Vector}
 Mat Instance Dynamic {type=MaterialInstanceDynamic}
 NotFrontEnd {type=number}
 Overlay Mat Instance Dynamic {type=MaterialInstanceDynamic}
 PlayerPawn {type=FortPlayerPawnAthena}
 R1 Velocity Transform Location {type=Vector}
 R1 Velocity Transform Velocity {type=Vector}
 R2 Velocity Transform Location {type=Vector}
 R2 Velocity Transform Velocity {type=Vector}
 SkelMesh {type=SkeletalMeshComponent}
 Velocity Interp Speed {type=number}
 WPO Pho Wind Intensity {type=number}
]
