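# Physics asset for the RainbowStraps female medium cosmetic's FaceAcc mesh, defining capsule bodies and constraints for its dynamic hair bones.
# Back, back-left, back-right and two side chains have limited swing and tight twist, hanging from master bodies joined to the head by free constraints.
# Spine, neck, head, clavicle and upper-arm capsules form a freely constrained upper body.
# Two-way free constraints cross-link the third back segment to its back-left and back-right neighbors.
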PhysicsAsset /BRCosmetics/Characters/Player/Female/Medium/Bodies/F_MED_RainbowStraps/Meshes/Parts/F_MED_RainbowStraps_FaceAcc_Physics
bodies {
 clavicle_l [{shapeType=Capsule, length=10, radius=5.5, shapeName=clavicle_l_capsule, center=<5 -0.8689 0>, rotation=<rot 90 0 0>}]
 clavicle_r [{shapeType=Capsule, length=10, radius=6, shapeName=clavicle_r_capsule, center=<-4.9679 0.5849 -0.5158>, rotation=<rot 82.7031 -0.0009 -0.001>}]
 dyn_hair_BL_01 [{shapeType=Capsule, length=4, radius=5, shapeName=dyn_hair_BL_01_capsule, center=<2 0 3.5>, rotation=<rot 0 90 90>}]
 dyn_hair_BL_02 [{shapeType=Capsule, length=4, radius=3.5, shapeName=dyn_hair_BL_02_capsule, center=<3 0 2>, rotation=<rot 0 90 90>}]
 dyn_hair_BL_03 [{shapeType=Capsule, length=5, radius=2.5, shapeName=dyn_hair_BL_03_capsule, center=<3 0 2>, rotation=<rot 0 90 90>}]
 dyn_hair_BL_Mstr [{shapeType=Capsule, length=0.505, radius=0.505, shapeName=dyn_hair_BL_Mstr_capsule, rotation=<rot 0 -90 0>}]
 dyn_hair_BR_01 [{shapeType=Capsule, length=4, radius=5, shapeName=dyn_hair_BR_01_capsule, center=<2 0 3>, rotation=<rot 0 90 90>}]
 dyn_hair_BR_02 [{shapeType=Capsule, length=4, radius=3, shapeName=dyn_hair_BR_02_capsule, center=<2 0 2.5>, rotation=<rot 0 90 101.2299>}]
 dyn_hair_BR_03 [{shapeType=Capsule, length=2, radius=3, shapeName=dyn_hair_BR_03_capsule, center=<2 0 2.5>, rotation=<rot 0 90 79.6018>}]
 dyn_hair_BR_04 [{shapeType=Capsule, length=4, radius=2.8, shapeName=dyn_hair_BR_04_capsule, center=<1 0 2>, rotation=<rot 0 90 90>}]
 dyn_hair_BR_Mstr [{shapeType=Capsule, length=0.505, radius=0.505, shapeName=dyn_hair_BR_Mstr_capsule, rotation=<rot 0 -90 0>}]
 dyn_hair_B_01 [{shapeType=Capsule, length=4, radius=5, shapeName=dyn_hair_B_01_capsule, center=<2 0 2.5>, rotation=<rot 0 90 90>}]
 dyn_hair_B_02 [{shapeType=Capsule, length=2, radius=3, shapeName=dyn_hair_B_02_capsule, center=<2 0 2>, rotation=<rot 0 90 90>}]
 dyn_hair_B_03 [{shapeType=Capsule, length=10, radius=2.5, shapeName=dyn_hair_B_03_capsule, center=<2 0 2>, rotation=<rot 0 0 90>}]
 dyn_hair_B_04 [{shapeType=Capsule, length=7, radius=2.5, shapeName=dyn_hair_B_04_capsule, center=<1 0 1.5>, rotation=<rot 0 0 90>}]
 dyn_hair_B_Mstr [{shapeType=Capsule, length=0.505, radius=0.505, shapeName=dyn_hair_B_Mstr_capsule, rotation=<rot 0 -90 0>}]
 dyn_hair_L_01_Mstr [{shapeType=Capsule, length=0.505, radius=0.505, shapeName=dyn_hair_L_01_Mstr_capsule, rotation=<rot 0 -90 0>}]
 dyn_hair_L_02_01 [{shapeType=Capsule, length=5, radius=2, shapeName=dyn_hair_L_02_01_capsule, center=<4 1 1>, rotation=<rot 0 90 90>}]
 dyn_hair_L_02_02 [{shapeType=Capsule, length=5, radius=1.5, shapeName=dyn_hair_L_02_02_capsule, center=<2 1 1>, rotation=<rot 0 90 90>}]
 dyn_hair_L_02_03 [{shapeType=Capsule, length=2, radius=1.5, shapeName=dyn_hair_L_02_03_capsule, center=<1 0 0>, rotation=<rot 0 90 90>}]
 dyn_hair_L_02_06 [{shapeType=Capsule, length=2, radius=1.5, shapeName=dyn_hair_L_02_06_capsule, center=<1 0 0>, rotation=<rot 0 90 90>}]
 dyn_hair_L_02_Mstr [{shapeType=Capsule, length=0.505, radius=0.505, shapeName=dyn_hair_L_02_Mstr_capsule, rotation=<rot 0 -90 0>}]
 dyn_hair_R_01_Mstr [{shapeType=Capsule, length=0.505, radius=0.505, shapeName=dyn_hair_R_01_Mstr_capsule, rotation=<rot 0 -90 0>}]
 dyn_hair_R_02_01 [{shapeType=Capsule, length=5, radius=2, shapeName=dyn_hair_R_02_01_capsule, center=<4 0 0>, rotation=<rot 0 90 90>}]
 dyn_hair_R_02_02 [{shapeType=Capsule, length=3, radius=1.5, shapeName=dyn_hair_R_02_02_capsule, center=<2 0 0>, rotation=<rot 0 90 90>}]
 dyn_hair_R_02_03 [{shapeType=Capsule, length=4, radius=2, shapeName=dyn_hair_R_02_03_capsule, center=<3 -1.5 0>, rotation=<rot 0 74.6098 90>}]
 dyn_hair_R_02_Mstr [{shapeType=Capsule, length=0.505, radius=0.505, shapeName=dyn_hair_R_02_Mstr_capsule, rotation=<rot 0 -90 0>}]
 head [{shapeType=Capsule, length=5, radius=4.5, center=<0.3099 2.8782 0>, rotation=<rot 0 79.5863 90>}]
 neck_01 [{shapeType=Capsule, length=6, radius=4, shapeName=neck_01_capsule, center=<3.1147 0.8814 0>, rotation=<rot 0 -90 90>}]
 spine_04 [{shapeType=Capsule, length=22, radius=8, shapeName=spine_05_capsule, center=<6.0666 0.2607 0>, rotation=<rot 0 -180 180>}]
 spine_05 [{shapeType=Capsule, length=22, radius=6, shapeName=spine_05_capsule, center=<5.3 -0.0502 0.296>, rotation=<rot 0 -180 180>}]
 upperarm_l [{shapeType=Capsule, length=15, radius=3, shapeName=upperarm_l_capsule, center=<9 0 0>, rotation=<rot 90 0 0>}]
 upperarm_r [{shapeType=Capsule, length=20, radius=3.5, shapeName=upperarm_r_capsule, center=<-13.4593 0 0>, rotation=<rot 90 0 0>}]
}
constraints [
 {bone1Name=dyn_hair_R_02_02, bone2Name=dyn_hair_R_02_01, swing1LimitDegrees=55, swing2LimitDegrees=35, twistLimitDegrees=5}
 {bone1Name=dyn_hair_B_02, bone2Name=dyn_hair_B_01, swing1LimitDegrees=20, swing2LimitDegrees=50, twistLimitDegrees=5}
 {bone1Name=dyn_hair_L_02_02, bone2Name=dyn_hair_L_02_01, swing1LimitDegrees=20, swing2LimitDegrees=10, twistLimitDegrees=5}
 {bone1Name=dyn_hair_R_02_03, bone2Name=dyn_hair_R_02_02, swing1LimitDegrees=45, swing2LimitDegrees=45, twistLimitDegrees=5}
 {bone1Name=dyn_hair_R_02_01, bone2Name=dyn_hair_R_02_Mstr, swing1LimitDegrees=20, swing2LimitDegrees=10, twistLimitDegrees=5}
 {
  bone1Name=dyn_hair_R_02_Mstr
  bone2Name=head
  swing1LimitDegrees=45
  swing1Motion=Free
  swing2LimitDegrees=45
  swing2Motion=Free
  twistLimitDegrees=45
  twistMotion=Free
 }
 {bone1Name=dyn_hair_L_02_03, bone2Name=dyn_hair_L_02_02, swing1LimitDegrees=25, swing2LimitDegrees=20, twistLimitDegrees=5}
 {bone1Name=dyn_hair_L_02_06, bone2Name=dyn_hair_L_02_02, swing1LimitDegrees=25, swing2LimitDegrees=20, twistLimitDegrees=5}
 {
  bone1Name=dyn_hair_R_01_Mstr
  bone2Name=head
  swing1LimitDegrees=45
  swing1Motion=Free
  swing2LimitDegrees=45
  swing2Motion=Free
  twistLimitDegrees=45
  twistMotion=Free
 }
 {bone1Name=dyn_hair_L_02_01, bone2Name=dyn_hair_L_02_Mstr, swing1LimitDegrees=25, swing2LimitDegrees=10, twistLimitDegrees=5}
 {bone1Name=dyn_hair_B_03, bone2Name=dyn_hair_B_02, swing1LimitDegrees=30, swing2LimitDegrees=65, twistLimitDegrees=5}
 {bone1Name=dyn_hair_B_04, bone2Name=dyn_hair_B_03, swing1LimitDegrees=45, swing2LimitDegrees=45, twistLimitDegrees=10}
 {
  bone1Name=dyn_hair_L_01_Mstr
  bone2Name=head
  swing1LimitDegrees=45
  swing1Motion=Free
  swing2LimitDegrees=45
  swing2Motion=Free
  twistLimitDegrees=45
  twistMotion=Free
 }
 {
  bone1Name=dyn_hair_L_02_Mstr
  bone2Name=head
  swing1LimitDegrees=45
  swing1Motion=Free
  swing2LimitDegrees=45
  swing2Motion=Free
  twistLimitDegrees=45
  twistMotion=Free
 }
 {bone1Name=dyn_hair_B_01, bone2Name=dyn_hair_B_Mstr, swing1LimitDegrees=20, swing2LimitDegrees=50, twistLimitDegrees=5}
 {
  bone1Name=dyn_hair_B_Mstr
  bone2Name=head
  swing1LimitDegrees=45
  swing1Motion=Free
  swing2LimitDegrees=45
  swing2Motion=Free
  twistLimitDegrees=45
  twistMotion=Free
 }
 {
  bone1Name=dyn_hair_BR_Mstr
  bone2Name=head
  swing1LimitDegrees=45
  swing1Motion=Free
  swing2LimitDegrees=45
  swing2Motion=Free
  twistLimitDegrees=45
  twistMotion=Free
 }
 {bone1Name=dyn_hair_BR_01, bone2Name=dyn_hair_BR_Mstr, swing1LimitDegrees=20, swing2LimitDegrees=45, twistLimitDegrees=5}
 {bone1Name=dyn_hair_BR_02, bone2Name=dyn_hair_BR_01, swing1LimitDegrees=20, swing2LimitDegrees=45, twistLimitDegrees=10}
 {bone1Name=dyn_hair_BR_03, bone2Name=dyn_hair_BR_02, swing1LimitDegrees=20, swing2LimitDegrees=45, twistLimitDegrees=10}
 {bone1Name=dyn_hair_BR_04, bone2Name=dyn_hair_BR_03, swing1LimitDegrees=35, swing2LimitDegrees=45, twistLimitDegrees=10}
 {bone1Name=dyn_hair_BL_01, bone2Name=dyn_hair_BL_Mstr, swing1LimitDegrees=20, swing2LimitDegrees=50, twistLimitDegrees=5}
 {bone1Name=dyn_hair_BL_02, bone2Name=dyn_hair_BL_01, swing1LimitDegrees=20, swing2LimitDegrees=50, twistLimitDegrees=5}
 {bone1Name=dyn_hair_BL_03, bone2Name=dyn_hair_BL_02, swing1LimitDegrees=30, swing2LimitDegrees=65, twistLimitDegrees=5}
 {
  bone1Name=dyn_hair_BL_Mstr
  bone2Name=head
  swing1LimitDegrees=45
  swing1Motion=Free
  swing2LimitDegrees=45
  swing2Motion=Free
  twistLimitDegrees=45
  twistMotion=Free
 }
 {
  bone1Name=head
  bone2Name=neck_01
  swing1LimitDegrees=45
  swing1Motion=Free
  swing2LimitDegrees=45
  swing2Motion=Free
  twistLimitDegrees=45
  twistMotion=Free
 }
 {
  bone1Name=clavicle_l
  bone2Name=spine_05
  swing1LimitDegrees=45
  swing1Motion=Free
  swing2LimitDegrees=45
  swing2Motion=Free
  twistLimitDegrees=45
  twistMotion=Free
 }
 {
  bone1Name=upperarm_l
  bone2Name=clavicle_l
  swing1LimitDegrees=45
  swing1Motion=Free
  swing2LimitDegrees=45
  swing2Motion=Free
  twistLimitDegrees=45
  twistMotion=Free
 }
 {
  bone1Name=clavicle_r
  bone2Name=spine_05
  swing1LimitDegrees=45
  swing1Motion=Free
  swing2LimitDegrees=45
  swing2Motion=Free
  twistLimitDegrees=45
  twistMotion=Free
 }
 {
  bone1Name=upperarm_r
  bone2Name=clavicle_r
  swing1LimitDegrees=45
  swing1Motion=Free
  swing2LimitDegrees=45
  swing2Motion=Free
  twistLimitDegrees=45
  twistMotion=Free
 }
 {
  bone1Name=spine_05
  bone2Name=spine_04
  swing1LimitDegrees=45
  swing1Motion=Free
  swing2LimitDegrees=45
  swing2Motion=Free
  twistLimitDegrees=45
  twistMotion=Free
 }
 {
  bone1Name=neck_01
  bone2Name=spine_05
  swing1LimitDegrees=45
  swing1Motion=Free
  swing2LimitDegrees=45
  swing2Motion=Free
  twistLimitDegrees=45
  twistMotion=Free
 }
 {
  bone1Name=dyn_hair_BL_03
  bone2Name=dyn_hair_B_03
  swing1LimitDegrees=45
  swing1Motion=Free
  swing2LimitDegrees=45
  swing2Motion=Free
  twistLimitDegrees=45
  twistMotion=Free
 }
 {
  bone1Name=dyn_hair_B_03
  bone2Name=dyn_hair_BL_03
  swing1LimitDegrees=45
  swing1Motion=Free
  swing2LimitDegrees=45
  swing2Motion=Free
  twistLimitDegrees=45
  twistMotion=Free
 }
 {
  bone1Name=dyn_hair_BR_03
  bone2Name=dyn_hair_B_03
  swing1LimitDegrees=45
  swing1Motion=Free
  swing2LimitDegrees=45
  swing2Motion=Free
  twistLimitDegrees=45
  twistMotion=Free
 }
 {
  bone1Name=dyn_hair_B_03
  bone2Name=dyn_hair_BR_03
  swing1LimitDegrees=45
  swing1Motion=Free
  swing2LimitDegrees=45
  swing2Motion=Free
  twistLimitDegrees=45
  twistMotion=Free
 }
]
