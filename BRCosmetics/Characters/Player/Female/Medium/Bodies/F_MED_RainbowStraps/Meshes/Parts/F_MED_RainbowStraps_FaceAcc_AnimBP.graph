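# Animation Blueprint for the RainbowStraps female FaceAcc cosmetic part, adding secondary motion to hair and a hat buckle.
# Body pose and head-mesh facial bones are copied, run through a Control Rig and cached.
# Bool blends choose among skydive, FE, default and pinned (spine/head-constrained) RigidBody simulations, then a long AnimDynamics chain with collision limits follows.
# Event updates set gravity override, head mesh, backpack state and physics-disable alphas.
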AnimBlueprint /BRCosmetics/Characters/Player/Female/Medium/Bodies/F_MED_RainbowStraps/Meshes/Parts/F_MED_RainbowStraps_FaceAcc_AnimBP
target_skeleton=&/F_MED_RainbowStraps_FaceAcc_Skeleton
graphs [
 AnimGraph {nodes [
   Anim Dynamics - Bone: dyn_hair_L_02_09(Alpha=0, ExternalForce="0,0,0") {
    angularDampingOverride=0.999
    bOverrideAngularDamping=true
    bUseGravityOverride=true
    bUseSphericalLimits=true
    gravityScale=0.35
    simulationSpace=RootRelative
    alphaBoolBlend {blendOption=Linear}
    boundBone {boneName=dyn_hair_L_02_09}
    componentAppliedLinearAccClamp=<3000 3000 3000>
    componentLinearAccScale=<0.3 0.3 1>
    componentLinearVelScale=<0.3 0.3 1>
    gravityOverride {z=-980}
    physicsBodyDefinitions [{
      boundBone {boneName=dyn_hair_L_02_09}
      boxExtents=<20 20 20>
      constraintSetup {angularLimitsMax=<3 25 55>, angularLimitsMin=<-3 -45 -55>, angularTarget=<1 0 0>}
      localJointOffset=<4 0 0>
     }]
    planarLimits [{drivingBone {boneName=spine_05}, planeTransform {location=<10 0 0>, rotation=<rot 0 -20 90>}}]
    relativeSpaceBone {boneName=dyn_hair_BL_03}
    sphericalLimits [
     {limitRadius=4, drivingBone {boneName=spine_05}, sphereLocalOffset=<9 0 -6>}
     {limitRadius=12, drivingBone {boneName=spine_04}, sphereLocalOffset=<5 0 -3>}
     {limitRadius=5, drivingBone {boneName=clavicle_l}, sphereLocalOffset=<7 0 2>}
     {limitRadius=7, drivingBone {boneName=spine_05}, sphereLocalOffset=<9 0 0>}
    ]
   }
   "Transform (Modify) Bone - Bone: dyn_hair_B_02"(Alpha=1, Translation="0,0,0", Rotation="0,0,0", Scale="1,1,1") {
    rotationMode=BMM_Additive
    rotationSpace=BCS_BoneSpace
    translationMode=BMM_Additive
    translationSpace=BCS_BoneSpace
    alphaBoolBlend {blendOption=Linear}
    boneToModify {boneName=dyn_hair_B_02}
   }
   n1 = RigidBody!(ComponentPose=Local To Component(LocalPose=Use cached pose 'PreDynamics'()), Alpha, OverrideWorldGravity:Vector="0,0,-1680", ExternalForce) -> (Pose) {
    bClampLinearTranslationLimitToRefPose=true
    bOverrideWorldGravity=true
    overridePhysicsAsset=&/F_MED_RainbowStraps_FaceAcc_Physics_FE
    simulationSpace=BaseBoneSpace
    alphaBoolBlend {blendOption=Linear}
    alphaScaleBiasClamp {bInterpResult=true, interpSpeedIncreasing=2}
    baseBoneRef {boneName=spine_05}
    componentAppliedLinearAccClamp=<4000 4000 4000>
    componentLinearAccScale=<0.3 0.3 0.5>
    componentLinearVelScale=<0.3 0.3 0.5>
    simSpaceSettings {
     maxAngularAcceleration=1000
     maxAngularVelocity=3
     maxLinearAcceleration=700
     maxLinearVelocity=500
     velocityScaleZ=0.5
     worldAlpha=0.5
     externalLinearDragV=<0.2 0.2>
    }
   }
   n2 = RigidBody(ComponentPose=Local To Component(LocalPose=Use cached pose 'PreDynamics'()), OverrideWorldGravity="0,0,-1680") {
    bClampLinearTranslationLimitToRefPose=true
    bOverrideWorldGravity=true
    overridePhysicsAsset=&/F_MED_RainbowStraps_FaceAcc_Physics
    simulationSpace=BaseBoneSpace
    alphaBoolBlend {blendOption=Linear}
    alphaScaleBiasClamp {bInterpResult=true, interpSpeedIncreasing=2}
    baseBoneRef {boneName=spine_05}
    componentAppliedLinearAccClamp=<4000 4000 4000>
    componentLinearAccScale=<0.3 0.3 0.5>
    componentLinearVelScale=<0.3 0.3 0.5>
    simSpaceSettings {
     maxAngularAcceleration=1000
     maxAngularVelocity=3
     maxLinearAcceleration=700
     maxLinearVelocity=500
     velocityScaleZ=0.5
     worldAlpha=0.5
     externalLinearDragV=<0.2 0.2>
    }
   }
   n3 = RigidBody(ComponentPose=Local To Component(LocalPose=Use cached pose 'PreDynamics'()), OverrideWorldGravity="0,0,-1680") {
    bClampLinearTranslationLimitToRefPose=true
    bOverrideWorldGravity=true
    overridePhysicsAsset=&/F_MED_RainbowStraps_FaceAcc_Physics_SkyDIve
    simulationSpace=BaseBoneSpace
    alphaBoolBlend {blendOption=Linear}
    alphaScaleBiasClamp {bInterpResult=true, interpSpeedIncreasing=2}
    baseBoneRef {boneName=spine_05}
    componentAppliedLinearAccClamp=<4000 4000 4000>
    componentLinearAccScale=<0.2 0.2 0.4>
    componentLinearVelScale=<0.2 0.2 0.4>
    simSpaceSettings {
     maxAngularAcceleration=1000
     maxAngularVelocity=3
     maxLinearAcceleration=700
     maxLinearVelocity=500
     velocityScaleZ=0.3
     worldAlpha=0.3
     externalLinearDragV=<0.2 0.2>
    }
   }
   n4 = Blend Poses by bool(
    BlendPose_0=Component To Local(ComponentPose=@n3)
    BlendPose_1=Blend Poses by bool(BlendPose_0=Component To Local(ComponentPose=@n1), BlendPose_1=Component To Local(ComponentPose=@n2)) {blendType=Linear}
   ) {blendType=Linear}
   n5 = "Constraint - dyn_hair_B_02 (PP)"!(ComponentPose=Local To Component(LocalPose=Use cached pose 'PreDynamics'()), Alpha=1, ConstraintWeights_0=0.9, ConstraintWeights_1:float=0.1) -> (Pose) {
    alphaBoolBlend {blendOption=Linear}
    boneToModify {boneName=dyn_hair_B_02}
    constraintSetup [{targetBone {boneName=spine_05}}, {targetBone {boneName=head}}]
    constraintWeights [1,1]
   }
   n6 = "Constraint - dyn_hair_BR_02 (PP)"!(ComponentPose=@n5, Alpha=1, ConstraintWeights_0=0.9, ConstraintWeights_1:float=0.1) -> (Pose) {
    alphaBoolBlend {blendOption=Linear}
    boneToModify {boneName=dyn_hair_BR_02}
    constraintSetup [{targetBone {boneName=spine_05}}, {targetBone {boneName=head}}]
    constraintWeights [1,1]
   }
   n7 = "Constraint - dyn_hair_BL_02 (PP)"!(ComponentPose=@n6, Alpha=1, ConstraintWeights_0=0.9, ConstraintWeights_1:float=0.1) -> (Pose) {
    alphaBoolBlend {blendOption=Linear}
    boneToModify {boneName=dyn_hair_BL_02}
    constraintSetup [{targetBone {boneName=spine_05}}, {targetBone {boneName=head}}]
    constraintWeights [1,1]
   }
   n8 = "Constraint - dyn_hair_B_03 (PP)"!(ComponentPose=@n7, Alpha=1, ConstraintWeights_0=1, ConstraintWeights_1:float=0) -> (Pose) {
    alphaBoolBlend {blendOption=Linear}
    boneToModify {boneName=dyn_hair_B_03}
    constraintSetup [{targetBone {boneName=spine_05}}, {targetBone {boneName=head}}]
    constraintWeights [1,1]
   }
   n9 = "Constraint - dyn_hair_BR_03 (PP)"!(ComponentPose=@n8, Alpha=1, ConstraintWeights_0=1, ConstraintWeights_1:float=0) -> (Pose) {
    alphaBoolBlend {blendOption=Linear}
    boneToModify {boneName=dyn_hair_BR_03}
    constraintSetup [{targetBone {boneName=spine_05}}, {targetBone {boneName=head}}]
    constraintWeights [1,1]
   }
   n10 = "Constraint - dyn_hair_BL_03 (PP)"!(ComponentPose=@n9, Alpha=1, ConstraintWeights_0=1, ConstraintWeights_1:float=0) -> (Pose) {
    alphaBoolBlend {blendOption=Linear}
    boneToModify {boneName=dyn_hair_BL_03}
    constraintSetup [{targetBone {boneName=spine_05}}, {targetBone {boneName=head}}]
    constraintWeights [1,1]
   }
   n11 = "Constraint - dyn_hair_R_02_02 (PP)"!(ComponentPose=@n10, Alpha=1, ConstraintWeights_0=0.3, ConstraintWeights_1:float=0.7) -> (Pose) {
    alphaBoolBlend {blendOption=Linear}
    boneToModify {boneName=dyn_hair_R_02_02}
    constraintSetup [{targetBone {boneName=head}}, {targetBone {boneName=spine_05}}]
    constraintWeights [1,1]
   }
   n12 = "Constraint - dyn_hair_R_02_03 (PP)"!(ComponentPose=@n11, Alpha=1, ConstraintWeights_0=0, ConstraintWeights_1:float=1) -> (Pose) {
    alphaBoolBlend {blendOption=Linear}
    boneToModify {boneName=dyn_hair_R_02_03}
    constraintSetup [{targetBone {boneName=head}}, {targetBone {boneName=spine_05}}]
    constraintWeights [1,1]
   }
   n13 = "Constraint - dyn_hair_L_02_02 (PP)"!(ComponentPose=@n12, Alpha=1, ConstraintWeights_0=0.3, ConstraintWeights_1:float=0.7) -> (Pose) {
    alphaBoolBlend {blendOption=Linear}
    boneToModify {boneName=dyn_hair_L_02_02}
    constraintSetup [{targetBone {boneName=head}}, {targetBone {boneName=spine_05}}]
    constraintWeights [1,1]
   }
   n14 = "Constraint - dyn_hair_L_02_02 (PP)"(ComponentPose=@n13, Alpha=1, ConstraintWeights_0=0, ConstraintWeights_1=1) {
    alphaBoolBlend {blendOption=Linear}
    boneToModify {boneName=dyn_hair_L_02_02}
    constraintSetup [{targetBone {boneName=head}}, {targetBone {boneName=spine_05}}]
    constraintWeights [1,1]
   }
   n15 = RigidBody(ComponentPose=@n14, OverrideWorldGravity="0,0,-1680") {
    bClampLinearTranslationLimitToRefPose=true
    bOverrideWorldGravity=true
    overridePhysicsAsset=&/F_MED_RainbowStraps_FaceAcc_Physics_Pinned
    simulationSpace=BaseBoneSpace
    alphaBoolBlend {blendOption=Linear}
    alphaScaleBiasClamp {bInterpResult=true, interpSpeedIncreasing=2}
    baseBoneRef {boneName=spine_05}
    componentAppliedLinearAccClamp=<4000 4000 4000>
    componentLinearAccScale=<0.3 0.3 0.5>
    componentLinearVelScale=<0.3 0.3 0.5>
    simSpaceSettings {
     maxAngularAcceleration=1000
     maxAngularVelocity=3
     maxLinearAcceleration=700
     maxLinearVelocity=500
     velocityScaleZ=0.5
     worldAlpha=0.5
     externalLinearDragV=<0.2 0.2>
    }
   }
   n16 = Component To Local(ComponentPose=@n15)
   n17 = Anim Dynamics - Bone: dyn_hair_R_01_01(
    ComponentPose=Local To Component(LocalPose=Blend Poses by bool(BlendPose_0=@n16, BlendPose_1=@n4) {blendType=Linear})
    Alpha=1
    ExternalForce="0,0,0"
   ) {
    angularDampingOverride=0.9
    angularSpringConstant=10000
    bOverrideAngularDamping=true
    bUseGravityOverride=true
    bUsePlanarLimit=false
    bUseSphericalLimits=true
    gravityScale=0.35
    simulationSpace=BoneRelative
    alphaBoolBlend {blendOption=Linear}
    boundBone {boneName=dyn_hair_R_01_01}
    componentAppliedLinearAccClamp=<3000 3000 3000>
    componentLinearAccScale=<0.2 0.2 0.7>
    componentLinearVelScale=<0.2 0.2 0.7>
    gravityOverride {z=-980}
    physicsBodyDefinitions [{
      boundBone {boneName=dyn_hair_R_01_01}
      boxExtents=<20 20 20>
      constraintSetup {angularLimitsMax=<3 5 10>, angularLimitsMin=<-3 -25 -25>, angularTarget=<1 0 0>}
      localJointOffset=<3 0 0>
     }]
    relativeSpaceBone {boneName=root}
   }
   n18 = Anim Dynamics - Bone: dyn_hair_R_01_02(ComponentPose=@n17, Alpha=1, ExternalForce="0,0,0") {
    angularDampingOverride=0.9
    angularSpringConstant=10000
    bOverrideAngularDamping=true
    bUseGravityOverride=true
    bUsePlanarLimit=false
    bUseSphericalLimits=true
    gravityScale=0.35
    simulationSpace=BoneRelative
    alphaBoolBlend {blendOption=Linear}
    boundBone {boneName=dyn_hair_R_01_02}
    componentAppliedLinearAccClamp=<3000 3000 3000>
    componentLinearAccScale=<0.2 0.2 0.7>
    componentLinearVelScale=<0.2 0.2 0.7>
    gravityOverride {z=-980}
    physicsBodyDefinitions [{
      boundBone {boneName=dyn_hair_R_01_02}
      boxExtents=<20 20 20>
      constraintSetup {angularLimitsMax=<3 15 20>, angularLimitsMin=<-3 -45 -25>, angularTarget=<1 0 0>}
      localJointOffset=<3 0 0>
     }]
    relativeSpaceBone {boneName=root}
    sphericalLimits [{limitRadius=8, drivingBone {boneName=head}, sphereLocalOffset=<0 0 0>}]
   }
   n19 = Anim Dynamics - Bone: dyn_hair_R_01_03(ComponentPose=@n18, Alpha=1, ExternalForce="0,0,0") {
    angularDampingOverride=0.9
    angularSpringConstant=10000
    bOverrideAngularDamping=true
    bUseGravityOverride=true
    bUsePlanarLimit=false
    bUseSphericalLimits=true
    gravityScale=0.35
    simulationSpace=BoneRelative
    alphaBoolBlend {blendOption=Linear}
    boundBone {boneName=dyn_hair_R_01_03}
    componentAppliedLinearAccClamp=<3000 3000 3000>
    componentLinearAccScale=<0.2 0.2 0.7>
    componentLinearVelScale=<0.2 0.2 0.7>
    gravityOverride {z=-980}
    physicsBodyDefinitions [{
      boundBone {boneName=dyn_hair_R_01_03}
      boxExtents=<20 20 20>
      constraintSetup {angularLimitsMax=<3 15 20>, angularLimitsMin=<-3 -45 -25>, angularTarget=<1 0 0>}
      localJointOffset=<3 0 0>
     }]
    relativeSpaceBone {boneName=root}
    sphericalLimits [{limitRadius=8, drivingBone {boneName=head}, sphereLocalOffset=<0 0 0>}]
   }
   n20 = Anim Dynamics - Bone: dyn_hair_L_01_01(ComponentPose=@n19, Alpha=1, ExternalForce="0,0,0") {
    angularDampingOverride=0.9
    angularSpringConstant=10000
    bOverrideAngularDamping=true
    bUseGravityOverride=true
    bUseSphericalLimits=true
    gravityScale=0.35
    simulationSpace=BoneRelative
    alphaBoolBlend {blendOption=Linear}
    boundBone {boneName=dyn_hair_L_01_01}
    componentAppliedLinearAccClamp=<3000 3000 3000>
    componentLinearAccScale=<0.2 0.2 0.7>
    componentLinearVelScale=<0.2 0.2 0.7>
    gravityOverride {z=-980}
    physicsBodyDefinitions [{
      boundBone {boneName=dyn_hair_L_01_01}
      boxExtents=<20 20 20>
      constraintSetup {angularLimitsMax=<3 5 10>, angularLimitsMin=<-3 -25 -25>, angularTarget=<1 0 0>}
      localJointOffset=<3 0 0>
     }]
    planarLimits [{drivingBone {boneName=head}, planeTransform {location=<0 7 -7>, rotation=<rot 0 0 120>}}]
    relativeSpaceBone {boneName=root}
    sphericalLimits [{limitRadius=4, drivingBone {boneName=head}, sphereLocalOffset=<-3 2 -8>}]
   }
   n21 = Anim Dynamics - Bone: dyn_hair_L_01_02(ComponentPose=@n20, Alpha=1, ExternalForce="0,0,0") {
    angularDampingOverride=0.9
    angularSpringConstant=10000
    bOverrideAngularDamping=true
    bUseGravityOverride=true
    bUseSphericalLimits=true
    gravityScale=0.35
    simulationSpace=BoneRelative
    alphaBoolBlend {blendOption=Linear}
    boundBone {boneName=dyn_hair_L_01_02}
    componentAppliedLinearAccClamp=<3000 3000 3000>
    componentLinearAccScale=<0.2 0.2 0.7>
    componentLinearVelScale=<0.2 0.2 0.7>
    gravityOverride {z=-980}
    physicsBodyDefinitions [{
      boundBone {boneName=dyn_hair_L_01_02}
      boxExtents=<20 20 20>
      constraintSetup {angularLimitsMax=<3 15 20>, angularLimitsMin=<-3 -45 -25>, angularTarget=<1 0 0>}
      localJointOffset=<3 0 0>
     }]
    planarLimits [{drivingBone {boneName=head}, planeTransform {location=<0 3 -3>, rotation=<rot 0 0 119.9999>}}]
    relativeSpaceBone {boneName=root}
    sphericalLimits [{limitRadius=4, drivingBone {boneName=head}, sphereLocalOffset=<-3 2 -8>}]
   }
   n22 = Anim Dynamics - Bone: dyn_hair_L_01_03(ComponentPose=@n21, Alpha=1, ExternalForce="0,0,0") {
    angularDampingOverride=0.9
    angularSpringConstant=10000
    bOverrideAngularDamping=true
    bUseGravityOverride=true
    bUseSphericalLimits=true
    gravityScale=0.35
    simulationSpace=BoneRelative
    alphaBoolBlend {blendOption=Linear}
    boundBone {boneName=dyn_hair_L_01_03}
    componentAppliedLinearAccClamp=<3000 3000 3000>
    componentLinearAccScale=<0.2 0.2 0.7>
    componentLinearVelScale=<0.2 0.2 0.7>
    gravityOverride {z=-980}
    physicsBodyDefinitions [{
      boundBone {boneName=dyn_hair_L_01_03}
      boxExtents=<20 20 20>
      constraintSetup {angularLimitsMax=<3 15 20>, angularLimitsMin=<-3 -45 -25>, angularTarget=<1 0 0>}
      localJointOffset=<3 0 0>
     }]
    planarLimits [{drivingBone {boneName=head}, planeTransform {location=<0 3 -3>, rotation=<rot 0 0 119.9999>}}]
    relativeSpaceBone {boneName=root}
    sphericalLimits [{limitRadius=4, drivingBone {boneName=head}, sphereLocalOffset=<-3 2 -8>}]
   }
   n23 = Anim Dynamics - Bone: dyn_hatbuckle_01(ComponentPose=@n22, Alpha=1, ExternalForce="0,0,0") {
    angularDampingOverride=0.9999
    angularSpringConstant=50
    bAngularSpring=true
    bOverrideAngularDamping=true
    bUsePlanarLimit=false
    simulationSpace=RootRelative
    alphaBoolBlend {blendOption=Linear}
    boundBone {boneName=dyn_hatbuckle_01}
    componentAppliedLinearAccClamp=<10000 10000 10000>
    componentLinearAccScale=<1 5 1>
    componentLinearVelScale=<1 5 1>
    physicsBodyDefinitions [{
      boundBone {boneName=dyn_hatbuckle_01}
      boxExtents=<12 12 12>
      constraintSetup {angularLimitsMax=<5 0 50>, angularLimitsMin=<-5 -60 -50>, angularTarget=<0 0 0>}
      localJointOffset=<2 0 0>
     }]
    relativeSpaceBone {boneName=spine_05}
   }
   n24 = Anim Dynamics - Bone: dyn_hair_BR_05(ComponentPose=@n23, Alpha=1, ExternalForce="0,0,0") {
    angularDampingOverride=1
    bOverrideAngularDamping=true
    bUseGravityOverride=true
    bUsePlanarLimit=false
    bUseSphericalLimits=true
    gravityScale=0.35
    simulationSpace=RootRelative
    alphaBoolBlend {blendOption=Linear}
    boundBone {boneName=dyn_hair_BR_05}
    componentAppliedLinearAccClamp=<3000 3000 3000>
    componentLinearAccScale=<0.3 0.3 2>
    componentLinearVelScale=<0.3 0.3 2>
    gravityOverride {z=-980}
    physicsBodyDefinitions [{
      boundBone {boneName=dyn_hair_BR_05}
      boxExtents=<10 10 10>
      constraintSetup {angularLimitsMax=<3 20 20>, angularLimitsMin=<-3 -45 -25>, angularTarget=<1 0 0>}
      localJointOffset=<2 0 0>
     }]
    relativeSpaceBone {boneName=dyn_hair_BR_04}
    sphericalLimits [
     {limitRadius=8, drivingBone {boneName=spine_05}, sphereLocalOffset=<3 0 4>}
     {limitRadius=8.5, drivingBone {boneName=spine_04}, sphereLocalOffset=<4 0 4>}
    ]
   }
   n25 = Anim Dynamics - Bone: dyn_hair_BR_06(ComponentPose=@n24, Alpha=1, ExternalForce="0,0,0") {
    angularDampingOverride=1
    bOverrideAngularDamping=true
    bUseGravityOverride=true
    bUsePlanarLimit=false
    bUseSphericalLimits=true
    gravityScale=0.35
    simulationSpace=RootRelative
    alphaBoolBlend {blendOption=Linear}
    boundBone {boneName=dyn_hair_BR_06}
    componentAppliedLinearAccClamp=<3000 3000 3000>
    componentLinearAccScale=<0.3 0.3 2>
    componentLinearVelScale=<0.3 0.3 2>
    gravityOverride {z=-980}
    physicsBodyDefinitions [{
      boundBone {boneName=dyn_hair_BR_06}
      boxExtents=<10 10 10>
      constraintSetup {angularLimitsMax=<3 20 20>, angularLimitsMin=<-3 -45 -25>, angularTarget=<1 0 0>}
      localJointOffset=<2 0 0>
     }]
    relativeSpaceBone {boneName=dyn_hair_BR_04}
    sphericalLimits [
     {limitRadius=8, drivingBone {boneName=spine_05}, sphereLocalOffset=<3 0 4>}
     {limitRadius=8.5, drivingBone {boneName=spine_04}, sphereLocalOffset=<4 0 4>}
    ]
   }
   n26 = Anim Dynamics - Bone: dyn_hair_B_05(ComponentPose=@n25, Alpha=1, ExternalForce="0,0,0") {
    angularDampingOverride=1
    bOverrideAngularDamping=true
    bUseGravityOverride=true
    bUsePlanarLimit=false
    bUseSphericalLimits=true
    gravityScale=0.35
    simulationSpace=RootRelative
    alphaBoolBlend {blendOption=Linear}
    boundBone {boneName=dyn_hair_B_05}
    componentAppliedLinearAccClamp=<3000 3000 3000>
    componentLinearAccScale=<0.3 0.3 2>
    componentLinearVelScale=<0.3 0.3 2>
    gravityOverride {z=-980}
    physicsBodyDefinitions [{
      boundBone {boneName=dyn_hair_B_05}
      boxExtents=<10 10 10>
      constraintSetup {angularLimitsMax=<3 20 20>, angularLimitsMin=<-3 -45 -25>, angularTarget=<1 0 0>}
      localJointOffset=<2 0 0>
     }]
    relativeSpaceBone {boneName=dyn_hair_B_04}
    sphericalLimits [
     {limitRadius=8, drivingBone {boneName=spine_05}, sphereLocalOffset=<3 0 0>}
     {limitRadius=8.5, drivingBone {boneName=spine_04}, sphereLocalOffset=<4 0 0>}
     {limitRadius=8, drivingBone {boneName=spine_05}, sphereLocalOffset=<3 0 5>}
    ]
   }
   n27 = Anim Dynamics - Bone: dyn_hair_B_06(ComponentPose=@n26, Alpha=1, ExternalForce="0,0,0") {
    angularDampingOverride=1
    bOverrideAngularDamping=true
    bUseGravityOverride=true
    bUsePlanarLimit=false
    bUseSphericalLimits=true
    gravityScale=0.35
    simulationSpace=RootRelative
    alphaBoolBlend {blendOption=Linear}
    boundBone {boneName=dyn_hair_B_06}
    componentAppliedLinearAccClamp=<3000 3000 3000>
    componentLinearAccScale=<0.3 0.3 2>
    componentLinearVelScale=<0.3 0.3 2>
    gravityOverride {z=-980}
    physicsBodyDefinitions [{
      boundBone {boneName=dyn_hair_B_06}
      boxExtents=<10 10 10>
      constraintSetup {angularLimitsMax=<3 20 20>, angularLimitsMin=<-3 -45 -25>, angularTarget=<1 0 0>}
      localJointOffset=<2 0 0>
     }]
    relativeSpaceBone {boneName=dyn_hair_B_04}
    sphericalLimits [
     {limitRadius=8, drivingBone {boneName=spine_05}, sphereLocalOffset=<3 0 0>}
     {limitRadius=8.5, drivingBone {boneName=spine_04}, sphereLocalOffset=<4 0 0>}
     {limitRadius=8, drivingBone {boneName=spine_05}, sphereLocalOffset=<3 0 5>}
    ]
   }
   n28 = Anim Dynamics - Bone: dyn_hair_BL_04(ComponentPose=@n27, Alpha=1, ExternalForce="0,0,0") {
    angularDampingOverride=1
    bOverrideAngularDamping=true
    bUseGravityOverride=true
    bUsePlanarLimit=false
    bUseSphericalLimits=true
    gravityScale=0.35
    simulationSpace=RootRelative
    alphaBoolBlend {blendOption=Linear}
    boundBone {boneName=dyn_hair_BL_04}
    componentAppliedLinearAccClamp=<3000 3000 3000>
    componentLinearAccScale=<0.3 0.3 2>
    componentLinearVelScale=<0.3 0.3 2>
    gravityOverride {z=-980}
    physicsBodyDefinitions [{
      boundBone {boneName=dyn_hair_BL_04}
      boxExtents=<10 10 10>
      constraintSetup {angularLimitsMax=<3 15 20>, angularLimitsMin=<-3 -45 -25>, angularTarget=<1 0 0>}
      localJointOffset=<2 0 0>
     }]
    relativeSpaceBone {boneName=dyn_hair_BL_03}
    sphericalLimits [
     {limitRadius=8.3, drivingBone {boneName=spine_05}, sphereLocalOffset=<4 0 -4>}
     {limitRadius=8.5, drivingBone {boneName=spine_04}, sphereLocalOffset=<4 0 -4>}
     {limitRadius=8.3, drivingBone {boneName=spine_05}, sphereLocalOffset=<4 0 3>}
    ]
   }
   n29 = Anim Dynamics - Bone: dyn_hair_BL_05(ComponentPose=@n28, Alpha=1, ExternalForce="0,0,0") {
    angularDampingOverride=1
    bOverrideAngularDamping=true
    bUseGravityOverride=true
    bUsePlanarLimit=false
    bUseSphericalLimits=true
    gravityScale=0.35
    simulationSpace=RootRelative
    alphaBoolBlend {blendOption=Linear}
    boundBone {boneName=dyn_hair_BL_05}
    componentAppliedLinearAccClamp=<3000 3000 3000>
    componentLinearAccScale=<0.3 0.3 2>
    componentLinearVelScale=<0.3 0.3 2>
    gravityOverride {z=-980}
    physicsBodyDefinitions [{
      boundBone {boneName=dyn_hair_BL_05}
      boxExtents=<10 10 10>
      constraintSetup {angularLimitsMax=<3 15 20>, angularLimitsMin=<-3 -45 -25>, angularTarget=<1 0 0>}
      localJointOffset=<2 0 0>
     }]
    relativeSpaceBone {boneName=dyn_hair_BL_03}
    sphericalLimits [
     {limitRadius=8.3, drivingBone {boneName=spine_05}, sphereLocalOffset=<4 0 -4>}
     {limitRadius=8.5, drivingBone {boneName=spine_04}, sphereLocalOffset=<4 0 -4>}
     {limitRadius=8.3, drivingBone {boneName=spine_05}, sphereLocalOffset=<4 0 3>}
    ]
   }
   n30 = Anim Dynamics - Bone: dyn_hair_R_02_01(ComponentPose=@n29, Alpha=1, ExternalForce="0,0,0") {
    angularDampingOverride=1
    bOverrideAngularDamping=true
    bUseGravityOverride=true
    bUsePlanarLimit=false
    bUseSphericalLimits=true
    gravityScale=0.35
    simulationSpace=RootRelative
    alphaBoolBlend {blendOption=Linear}
    boundBone {boneName=dyn_hair_R_02_01}
    componentAppliedLinearAccClamp=<3000 3000 3000>
    componentLinearAccScale=<0.3 0.3 1>
    componentLinearVelScale=<0.3 0.3 1>
    gravityOverride {z=-980}
    physicsBodyDefinitions [{
      boundBone {boneName=dyn_hair_R_02_01}
      boxExtents=<20 20 20>
      constraintSetup {angularLimitsMax=<3 30 20>, angularLimitsMin=<-3 -45 -25>, angularTarget=<1 0 0>}
      localJointOffset=<4 0 0>
     }]
    planarLimits [{drivingBone {boneName=spine_05}, planeTransform {location=<10 0 0>, rotation=<rot 0 -20 90>}}]
    relativeSpaceBone {boneName=dyn_hair_BL_03}
    sphericalLimits [
     {limitRadius=6, drivingBone {boneName=spine_05}, sphereLocalOffset=<5 0 7>}
     {limitRadius=8, drivingBone {boneName=clavicle_r}, sphereLocalOffset=<-5 0 0>}
     {limitRadius=7, drivingBone {boneName=head}, sphereLocalOffset=<0 0 0>}
     {limitRadius=7, drivingBone {boneName=upperarm_r}, sphereLocalOffset=<-9 0 0>}
    ]
   }
   n31 = Anim Dynamics - Bone: dyn_hair_R_02_02(ComponentPose=@n30, Alpha=1, ExternalForce="0,0,0") {
    angularDampingOverride=1
    bOverrideAngularDamping=true
    bUseGravityOverride=true
    bUsePlanarLimit=false
    bUseSphericalLimits=true
    gravityScale=0.35
    simulationSpace=RootRelative
    alphaBoolBlend {blendOption=Linear}
    boundBone {boneName=dyn_hair_R_02_02}
    componentAppliedLinearAccClamp=<3000 3000 3000>
    componentLinearAccScale=<0.3 0.3 1>
    componentLinearVelScale=<0.3 0.3 1>
    gravityOverride {z=-980}
    physicsBodyDefinitions [{
      boundBone {boneName=dyn_hair_R_02_02}
      boxExtents=<20 20 20>
      constraintSetup {angularLimitsMax=<3 30 20>, angularLimitsMin=<-3 -45 -25>, angularTarget=<1 0 0>}
      localJointOffset=<4 0 0>
     }]
    planarLimits [{drivingBone {boneName=spine_05}, planeTransform {location=<10 0 0>, rotation=<rot 0 -20 90>}}]
    relativeSpaceBone {boneName=dyn_hair_BL_03}
    sphericalLimits [
     {limitRadius=6.5, drivingBone {boneName=spine_05}, sphereLocalOffset=<5 0 7>}
     {limitRadius=7, drivingBone {boneName=clavicle_r}, sphereLocalOffset=<-5.5 0 0>}
     {limitRadius=7, drivingBone {boneName=head}, sphereLocalOffset=<0 0 0>}
     {limitRadius=7, drivingBone {boneName=upperarm_r}, sphereLocalOffset=<-9 0 0>}
     {limitRadius=7, drivingBone {boneName=neck_01}, sphereLocalOffset=<0 0 0>}
    ]
   }
   n32 = Anim Dynamics - Bone: dyn_hair_R_02_03(ComponentPose=@n31, Alpha=1, ExternalForce="0,0,0") {
    angularDampingOverride=1
    bOverrideAngularDamping=true
    bUseGravityOverride=true
    bUseSphericalLimits=true
    gravityScale=0.35
    simulationSpace=RootRelative
    alphaBoolBlend {blendOption=Linear}
    boundBone {boneName=dyn_hair_R_02_03}
    componentAppliedLinearAccClamp=<3000 3000 3000>
    componentLinearAccScale=<0.3 0.3 1>
    componentLinearVelScale=<0.3 0.3 1>
    gravityOverride {z=-980}
    physicsBodyDefinitions [{
      boundBone {boneName=dyn_hair_R_02_03}
      boxExtents=<20 20 20>
      constraintSetup {angularLimitsMax=<3 30 20>, angularLimitsMin=<-3 -45 -25>, angularTarget=<1 0 0>}
      localJointOffset=<4 0 0>
     }]
    planarLimits [{drivingBone {boneName=spine_05}, planeTransform {location=<10 0 0>, rotation=<rot 0 -20 90>}}]
    relativeSpaceBone {boneName=dyn_hair_BL_03}
    sphericalLimits [
     {limitRadius=7.4, drivingBone {boneName=spine_05}, sphereLocalOffset=<5 -2 2>}
     {limitRadius=7, drivingBone {boneName=clavicle_r}, sphereLocalOffset=<-5 2 0>}
     {limitRadius=7, drivingBone {boneName=neck_01}, sphereLocalOffset=<0 0 0>}
     {limitRadius=7, drivingBone {boneName=upperarm_r}, sphereLocalOffset=<-9 0 0>}
    ]
   }
   n33 = Anim Dynamics - Bone: dyn_hair_L_02_01(ComponentPose=@n32, Alpha=1, ExternalForce="0,0,0") {
    angularDampingOverride=0.999
    bOverrideAngularDamping=true
    bUseGravityOverride=true
    bUseSphericalLimits=true
    gravityScale=0.35
    simulationSpace=RootRelative
    alphaBoolBlend {blendOption=Linear}
    boundBone {boneName=dyn_hair_L_02_01}
    componentAppliedLinearAccClamp=<3000 3000 3000>
    componentLinearAccScale=<0.3 0.3 1>
    componentLinearVelScale=<0.3 0.3 1>
    gravityOverride {z=-980}
    physicsBodyDefinitions [{
      boundBone {boneName=dyn_hair_L_02_01}
      boxExtents=<20 20 20>
      constraintSetup {angularLimitsMax=<3 25 55>, angularLimitsMin=<-3 -45 -55>, angularTarget=<1 0 0>}
      localJointOffset=<5 0 0>
     }]
    planarLimits [{drivingBone {boneName=spine_05}, planeTransform {location=<10 0 0>, rotation=<rot 0 -20 90>}}]
    relativeSpaceBone {boneName=dyn_hair_BL_03}
    sphericalLimits [
     {limitRadius=5, drivingBone {boneName=spine_05}, sphereLocalOffset=<9 0 -6>}
     {limitRadius=0, drivingBone {boneName=None}, sphereLocalOffset=<0 0 0>}
     {limitRadius=5, drivingBone {boneName=clavicle_l}, sphereLocalOffset=<7 0 2>}
     {limitRadius=7, drivingBone {boneName=spine_05}, sphereLocalOffset=<9 0 0>}
     {limitRadius=7, drivingBone {boneName=neck_01}, sphereLocalOffset=<4 3.5 0>}
     {limitRadius=7, drivingBone {boneName=upperarm_l}, sphereLocalOffset=<0 0 0>}
    ]
   }
   n34 = Anim Dynamics - Bone: dyn_hair_L_02_02(ComponentPose=@n33, Alpha=1, ExternalForce="0,0,0") {
    angularDampingOverride=0.999
    bOverrideAngularDamping=true
    bUseGravityOverride=true
    bUseSphericalLimits=true
    gravityScale=0.35
    simulationSpace=RootRelative
    alphaBoolBlend {blendOption=Linear}
    boundBone {boneName=dyn_hair_L_02_02}
    componentAppliedLinearAccClamp=<3000 3000 3000>
    componentLinearAccScale=<0.3 0.3 1>
    componentLinearVelScale=<0.3 0.3 1>
    gravityOverride {z=-980}
    physicsBodyDefinitions [{
      boundBone {boneName=dyn_hair_L_02_02}
      boxExtents=<20 20 20>
      constraintSetup {angularLimitsMax=<3 25 55>, angularLimitsMin=<-3 -45 -55>, angularTarget=<1 0 0>}
      localJointOffset=<5 0 0>
     }]
    planarLimits [{drivingBone {boneName=spine_05}, planeTransform {location=<10 0 0>, rotation=<rot 0 -20 90>}}]
    relativeSpaceBone {boneName=dyn_hair_BL_03}
    sphericalLimits [
     {limitRadius=5, drivingBone {boneName=spine_05}, sphereLocalOffset=<9 0 -6>}
     {limitRadius=5, drivingBone {boneName=clavicle_l}, sphereLocalOffset=<7 0 2>}
     {limitRadius=7, drivingBone {boneName=spine_05}, sphereLocalOffset=<9 0 0>}
     {limitRadius=6, drivingBone {boneName=neck_01}, sphereLocalOffset=<4 2 0>}
     {limitRadius=7, drivingBone {boneName=upperarm_l}, sphereLocalOffset=<0 0 0>}
    ]
   }
   n35 = Anim Dynamics - Bone: dyn_hair_L_02_03(ComponentPose=@n34, Alpha=1, ExternalForce="0,0,0") {
    angularDampingOverride=0.999
    bOverrideAngularDamping=true
    bUseGravityOverride=true
    bUseSphericalLimits=true
    gravityScale=0.35
    simulationSpace=RootRelative
    alphaBoolBlend {blendOption=Linear}
    boundBone {boneName=dyn_hair_L_02_03}
    componentAppliedLinearAccClamp=<3000 3000 3000>
    componentLinearAccScale=<0.3 0.3 1>
    componentLinearVelScale=<0.3 0.3 1>
    gravityOverride {z=-980}
    physicsBodyDefinitions [{
      boundBone {boneName=dyn_hair_L_02_03}
      boxExtents=<20 20 20>
      constraintSetup {angularLimitsMax=<3 25 55>, angularLimitsMin=<-3 -45 -55>, angularTarget=<1 0 0>}
      localJointOffset=<5 0 0>
     }]
    planarLimits [{drivingBone {boneName=spine_05}, planeTransform {location=<10 0 0>, rotation=<rot 0 -20 90>}}]
    relativeSpaceBone {boneName=dyn_hair_BL_03}
    sphericalLimits [
     {limitRadius=5, drivingBone {boneName=spine_05}, sphereLocalOffset=<9 0 -6>}
     {limitRadius=0, drivingBone {boneName=None}, sphereLocalOffset=<0 0 0>}
     {limitRadius=5, drivingBone {boneName=clavicle_l}, sphereLocalOffset=<7 0 2>}
     {limitRadius=7, drivingBone {boneName=spine_05}, sphereLocalOffset=<9 0 0>}
     {limitRadius=6, drivingBone {boneName=neck_01}, sphereLocalOffset=<0 0 0>}
     {limitRadius=7, drivingBone {boneName=upperarm_l}, sphereLocalOffset=<0 0 0>}
    ]
   }
   n36 = Anim Dynamics - Bone: dyn_hair_R_02_04(ComponentPose=@n35, Alpha=1, ExternalForce="0,0,0") {
    angularDampingOverride=1
    bOverrideAngularDamping=true
    bUseGravityOverride=true
    bUsePlanarLimit=false
    bUseSphericalLimits=true
    gravityScale=0.35
    simulationSpace=RootRelative
    alphaBoolBlend {blendOption=Linear}
    boundBone {boneName=dyn_hair_R_02_04}
    componentAppliedLinearAccClamp=<3000 3000 3000>
    componentLinearAccScale=<0.3 0.3 1>
    componentLinearVelScale=<0.3 0.3 1>
    gravityOverride {z=-980}
    physicsBodyDefinitions [{
      boundBone {boneName=dyn_hair_R_02_04}
      boxExtents=<20 20 20>
      constraintSetup {angularLimitsMax=<3 30 20>, angularLimitsMin=<-3 -45 -25>, angularTarget=<1 0 0>}
      localJointOffset=<4 0 0>
     }]
    planarLimits [{drivingBone {boneName=spine_05}, planeTransform {location=<21 0 0>, rotation=<rot 0 -27 90>}}]
    relativeSpaceBone {boneName=dyn_hair_BL_03}
    sphericalLimits [
     {limitRadius=6, drivingBone {boneName=spine_05}, sphereLocalOffset=<5 0 5>}
     {limitRadius=10.3, drivingBone {boneName=spine_04}, sphereLocalOffset=<2 3 3>}
     {limitRadius=7.7, drivingBone {boneName=clavicle_r}, sphereLocalOffset=<-5 0 0>}
    ]
   }
   n37 = Anim Dynamics - Bone: dyn_hair_R_02_06(ComponentPose=@n36, Alpha=1, ExternalForce="0,0,0") {
    angularDampingOverride=1
    bOverrideAngularDamping=true
    bUseGravityOverride=true
    bUsePlanarLimit=false
    bUseSphericalLimits=true
    gravityScale=0.35
    simulationSpace=RootRelative
    alphaBoolBlend {blendOption=Linear}
    boundBone {boneName=dyn_hair_R_02_06}
    componentAppliedLinearAccClamp=<3000 3000 3000>
    componentLinearAccScale=<0.3 0.3 1>
    componentLinearVelScale=<0.3 0.3 1>
    gravityOverride {z=-980}
    physicsBodyDefinitions [{
      boundBone {boneName=dyn_hair_R_02_06}
      boxExtents=<20 20 20>
      constraintSetup {angularLimitsMax=<3 30 20>, angularLimitsMin=<-3 -45 -25>, angularTarget=<1 0 0>}
      localJointOffset=<4 0 0>
     }]
    planarLimits [{drivingBone {boneName=spine_05}, planeTransform {location=<21 0 0>, rotation=<rot 0 -27 90>}}]
    relativeSpaceBone {boneName=dyn_hair_BL_03}
    sphericalLimits [
     {limitRadius=6.3, drivingBone {boneName=spine_05}, sphereLocalOffset=<5 0 5>}
     {limitRadius=10.1, drivingBone {boneName=spine_04}, sphereLocalOffset=<5 1.5 2>}
     {limitRadius=5, drivingBone {boneName=spine_04}, sphereLocalOffset=<5 1.5 11>}
     {limitRadius=7.7, drivingBone {boneName=clavicle_r}, sphereLocalOffset=<-5 0 0>}
     {limitRadius=6, drivingBone {boneName=upperarm_r}, sphereLocalOffset=<-5 0 0>}
    ]
   }
   n38 = Anim Dynamics - Bone: dyn_hair_R_02_08(ComponentPose=@n37, Alpha=1, ExternalForce="0,0,0") {
    angularDampingOverride=1
    bOverrideAngularDamping=true
    bUseGravityOverride=true
    bUsePlanarLimit=false
    bUseSphericalLimits=true
    gravityScale=0.35
    simulationSpace=RootRelative
    alphaBoolBlend {blendOption=Linear}
    boundBone {boneName=dyn_hair_R_02_08}
    componentAppliedLinearAccClamp=<3000 3000 3000>
    componentLinearAccScale=<0.3 0.3 1>
    componentLinearVelScale=<0.3 0.3 1>
    gravityOverride {z=-980}
    physicsBodyDefinitions [{
      boundBone {boneName=dyn_hair_R_02_08}
      boxExtents=<20 20 20>
      constraintSetup {angularLimitsMax=<3 30 20>, angularLimitsMin=<-3 -45 -25>, angularTarget=<1 0 0>}
      localJointOffset=<5 0 0>
     }]
    planarLimits [{drivingBone {boneName=spine_05}, planeTransform {location=<21 0 0>, rotation=<rot 0 -27 90>}}]
    relativeSpaceBone {boneName=dyn_hair_BL_03}
    sphericalLimits [
     {limitRadius=6, drivingBone {boneName=spine_05}, sphereLocalOffset=<5 0 5>}
     {limitRadius=10, drivingBone {boneName=spine_04}, sphereLocalOffset=<7 1.5 2>}
     {limitRadius=7.7, drivingBone {boneName=clavicle_r}, sphereLocalOffset=<-5 0 0>}
     {limitRadius=6, drivingBone {boneName=upperarm_r}, sphereLocalOffset=<-5 0 0>}
    ]
   }
   n39 = Anim Dynamics - Bone: dyn_hair_L_02_03(ComponentPose=@n38, Alpha=1, ExternalForce="0,0,0") {
    angularDampingOverride=0.999
    bOverrideAngularDamping=true
    bUseGravityOverride=true
    bUseSphericalLimits=true
    gravityScale=0.35
    simulationSpace=RootRelative
    alphaBoolBlend {blendOption=Linear}
    boundBone {boneName=dyn_hair_L_02_03}
    componentAppliedLinearAccClamp=<3000 3000 3000>
    componentLinearAccScale=<0.3 0.3 1>
    componentLinearVelScale=<0.3 0.3 1>
    gravityOverride {z=-980}
    physicsBodyDefinitions [{
      boundBone {boneName=dyn_hair_L_02_03}
      boxExtents=<20 20 20>
      constraintSetup {angularLimitsMax=<3 25 55>, angularLimitsMin=<-3 -45 -55>, angularTarget=<1 0 0>}
      localJointOffset=<5 0 0>
     }]
    planarLimits [{drivingBone {boneName=spine_05}, planeTransform {location=<10 0 0>, rotation=<rot 0 -20 90>}}]
    relativeSpaceBone {boneName=dyn_hair_BL_03}
    sphericalLimits [
     {limitRadius=5, drivingBone {boneName=spine_05}, sphereLocalOffset=<9 0 -6>}
     {limitRadius=12, drivingBone {boneName=spine_04}, sphereLocalOffset=<6 0 -5>}
     {limitRadius=5, drivingBone {boneName=clavicle_l}, sphereLocalOffset=<7 0 2>}
     {limitRadius=7, drivingBone {boneName=spine_05}, sphereLocalOffset=<9 0 0>}
    ]
   }
   n40 = Anim Dynamics - Bone: dyn_hair_L_02_06(ComponentPose=@n39, Alpha=1, ExternalForce="0,0,0") {
    angularDampingOverride=0.999
    bOverrideAngularDamping=true
    bUseGravityOverride=true
    bUseSphericalLimits=true
    gravityScale=0.35
    simulationSpace=RootRelative
    alphaBoolBlend {blendOption=Linear}
    boundBone {boneName=dyn_hair_L_02_06}
    componentAppliedLinearAccClamp=<3000 3000 3000>
    componentLinearAccScale=<0.3 0.3 1>
    componentLinearVelScale=<0.3 0.3 1>
    gravityOverride {z=-980}
    physicsBodyDefinitions [{
      boundBone {boneName=dyn_hair_L_02_06}
      boxExtents=<20 20 20>
      constraintSetup {angularLimitsMax=<3 25 55>, angularLimitsMin=<-3 -45 -55>, angularTarget=<1 0 0>}
      localJointOffset=<4 0 0>
     }]
    planarLimits [{drivingBone {boneName=spine_05}, planeTransform {location=<10 0 0>, rotation=<rot 0 -20 90>}}]
    relativeSpaceBone {boneName=dyn_hair_BL_03}
    sphericalLimits [
     {limitRadius=6, drivingBone {boneName=spine_05}, sphereLocalOffset=<9 0 -7>}
     {limitRadius=12, drivingBone {boneName=spine_04}, sphereLocalOffset=<5 0 -3>}
     {limitRadius=5, drivingBone {boneName=clavicle_l}, sphereLocalOffset=<7 0 2>}
     {limitRadius=7, drivingBone {boneName=spine_05}, sphereLocalOffset=<9 0 -3>}
    ]
   }
   n41 = "Constraint - dyn_hair_L_02_09 (P)"(ComponentPose=@n40, Alpha=1, ConstraintWeights_0=1) {
    alphaBoolBlend {blendOption=Linear}
    boneToModify {boneName=dyn_hair_L_02_09}
    constraintSetup [{targetBone {boneName=dyn_hair_L_02_06}}]
    constraintWeights [1]
   }
   Output Pose(Result=Component To Local(ComponentPose=@n41)) {layerGroup=None}
   n42 = Get Head Mesh() -> (Head Mesh:Skeletal Mesh Component Object Reference)
   n43 = Control Rig!(
    Source=Layered blend per bone(BasePose=Copy Pose From Mesh() {bUseAttachedParent=true}, BlendPoses_0=Copy Pose From Mesh(SourceMeshComponent=@n42)) {layerSetup [{branchFilters [{boneName=FACIAL_C_FacialRoot}, {boneName=FACIAL_C_Jaw}]}]}
    Alpha
    IsCrouchCtrl:bool=false
    IsSkyDiveingCtrl:bool=false
   ) -> (Pose) {
    alphaBoolBlend {blendOption=Linear}
    controlRigAssetReference {blueprintRigClass=&/F_MED_RainbowStraps_FaceAcc_CtrlRig.F_MED_RainbowStraps_FaceAcc_CtrlRig_C}
   }
   Save cached pose 'PreDynamics'(Pose=@n43) {cacheName=PreDynamics}
  ]}
 EventGraph {nodes [
   Event BlueprintUpdateAnimation()
   > n1 = Sequence!(execute) -> (then_0, then_1, then_2:Exec)
   > n2 = Gravity Override State Function DEPRECATED(CharacterPartAnimInstance=Self-Reference(), GameTimeInSeconds=GetGameTimeInSeconds(), GravityOverrideParams=Get Gravity Override Params())
   > Set Gravity Override(Gravity Override=@n2)
   n3 = Cast To PlayerPawn_Athena(execute=@n1.then_1, Object=GetOwningActor())
   > Set Head Mesh(Head Mesh:Skeletal Mesh Component Object Reference=GetSkeletalMeshForPartType(self=@n3, PartType=Head)) -> (Output_Get:Skeletal Mesh Component Object Reference)
   n4 = CheckForBackbling(self=Self-Reference())
   n5 = "Not Equal (Boolean)"(A=Get bRigidBodyEnabled(), B=true) -> (ReturnValue=~)
   Set wearing backpack Bool(execute=@n1.then_2, wearing backpack Bool:bool=OR Boolean(A=OR Boolean(A=@n4, B=@n4.bIsBackblingACape), B=@n5)) -> (Output_Get:bool)
   > Set NoRbanAlpha(NoRbanAlpha:double=SelectFloat(A=1, B=0, bPickA=@n5)) -> (Output_Get:double)
   n6 = Get bAnimDynamicsEnabled() -> (bAnimDynamicsEnabled:bool)
   > Set NoDynamicAlpha(NoDynamicAlpha:double=SelectFloat(A=1, B=0, bPickA="Not Equal (Boolean)"(A=@n6, B=true) -> (ReturnValue=~))) -> (Output_Get:double)
  ]}
]
variables [
 Gravity Override {type=Vector}
 Gravity Override Params {type=GravityOverrideParamsStruct}
 Head Mesh {type=SkeletalMeshComponent}
 NoDynamicAlpha {type=number}
 NoRbanAlpha {type=number}
 wearing backpack Bool {type=boolean}
]
